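# Physics asset for the purple leather jacket coat on the large female player body, an emote variant by its name.
# Capsules cover pelvis, spine, thighs and seven coat chains: front, under-arm and back on each side, plus mid-back.
# Chains join spine_04 through free-swinging master bones, with mostly limited joints and a locked link into each lower dyn_ section.
# Cross-constraints tie adjacent chains: front to under-arm, and mid-back to both side backs.
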PhysicsAsset /BRCosmetics/Characters/Player/Female/Large/Bodies/F_LRG_Leather_Jacket_Purple/Meshes/F_LRG_Leather_Jacket_Purple_Coat_Phys1_emote
bodies {
 back_01_l [{shapeType=Capsule, length=8, radius=7, shapeName=frnt_torso_02_r_capsule, center=<8 6 0>, rotation=<rot 90 0 0>}]
 back_01_r [{shapeType=Capsule, length=8, radius=7, shapeName=frnt_torso_02_r_capsule, center=<-8 -6 0>, rotation=<rot 90 0 0>}]
 back_02_l [{shapeType=Capsule, length=9, radius=7, shapeName=frnt_torso_02_r_capsule, center=<9 6 0>, rotation=<rot 90 0 0>}]
 back_02_r [{shapeType=Capsule, length=9, radius=7, shapeName=frnt_torso_02_r_capsule, center=<-9 -6 0>, rotation=<rot 90 0 0>}]
 dyn_back_01_l [{shapeType=Capsule, length=12, radius=7, shapeName=dyn_back_01_l_capsule, center=<10 6 0>, rotation=<rot 0 0 -10>}]
 dyn_back_01_r [{shapeType=Capsule, length=12, radius=7, shapeName=dyn_back_01_r_capsule, center=<-10 -6 0>, rotation=<rot 0 0 -10>}]
 dyn_back_02_l [
  {shapeType=Capsule, length=12, radius=7, shapeName=dyn_under_arm_01_l_capsule, center=<10 6 0>, rotation=<rot 0 0 -10>}
  {shapeType=Capsule, length=15, radius=5, shapeName=dyn_back_02_l_capsule, center=<10 6 -6>, rotation=<rot 0 0 25>}
 ]
 dyn_back_02_r [
  {shapeType=Capsule, length=12, radius=7, shapeName=dyn_back_01_r_capsule, center=<-10 -6 0>, rotation=<rot 0 0 -10>}
  {shapeType=Capsule, length=15, radius=5, shapeName=dyn_back_02_r_capsule, center=<-10 -6 6>, rotation=<rot 0 0 25>}
 ]
 dyn_back_03_l [{shapeType=Capsule, length=12, radius=4, shapeName=dyn_back_03_l_capsule, center=<7 0 0>, rotation=<rot 0 0 -10>}]
 dyn_back_03_r [{shapeType=Capsule, length=12, radius=4, shapeName=dyn_back_03_r_capsule, center=<-7 0 0>, rotation=<rot 0 0 -10>}]
 dyn_frnt_torso_01_l [{shapeType=Capsule, length=18, radius=8, shapeName=dyn_frnt_torso_01_l_capsule, center=<10 6 0>}]
 dyn_frnt_torso_01_r [{shapeType=Capsule, length=18, radius=8, shapeName=dyn_frnt_torso_01_r_capsule, center=<-10 -6 0>}]
 dyn_frnt_torso_02_l [{shapeType=Capsule, length=18, radius=8, shapeName=dyn_under_arm_01_l_capsule, center=<10 6 0>}]
 dyn_frnt_torso_02_r [{shapeType=Capsule, length=18, radius=8, shapeName=dyn_back_01_r_capsule, center=<-10 -6 0>}]
 dyn_frnt_torso_03_l [{shapeType=Capsule, length=16, radius=5, shapeName=dyn_frnt_torso_03_l_capsule, center=<7 4 0>}]
 dyn_frnt_torso_03_r [{shapeType=Capsule, length=16, radius=5, shapeName=dyn_frnt_torso_03_r_capsule, center=<-7 -4 0>}]
 dyn_mid_back_01 [{shapeType=Capsule, length=12, radius=7, shapeName=dyn_mid_back_01_capsule, center=<10 -6 0>}]
 dyn_mid_back_02 [{shapeType=Capsule, length=12, radius=7, shapeName=dyn_mid_back_01_capsule, center=<10 -6 0>}]
 dyn_mid_back_03 [{shapeType=Capsule, length=12, radius=4, shapeName=dyn_mid_back_03_capsule, center=<7 0 0>}]
 dyn_mstr_back_l [{shapeType=Capsule, length=10, radius=7, shapeName=dyn_mstr_back_l_capsule, center=<10 6 0>, rotation=<rot 0 0 -10>}]
 dyn_mstr_back_r [{shapeType=Capsule, length=10, radius=7, shapeName=dyn_mstr_back_r_capsule, center=<-10 -6 0>, rotation=<rot 0 0 -10>}]
 dyn_mstr_frnt_torso_l [{shapeType=Capsule, length=10, radius=7, shapeName=dyn_mstr_frnt_torso_l_capsule, center=<10 6 0>}]
 dyn_mstr_frnt_torso_r [{shapeType=Capsule, length=10, radius=7, shapeName=dyn_mstr_frnt_torso_r_capsule, center=<-10 -6 0>}]
 dyn_mstr_mid_back [{shapeType=Capsule, length=10, radius=7, shapeName=dyn_mstr_mid_back_capsule, center=<12 -6 0>}]
 dyn_mstr_under_arm_l [{shapeType=Capsule, length=10, radius=7, shapeName=dyn_mstr_under_arm_l_capsule, center=<10 6 0>}]
 dyn_mstr_under_arm_r [{shapeType=Capsule, length=10, radius=7, shapeName=dyn_mstr_under_arm_r_capsule, center=<-10 -6 0>}]
 dyn_under_arm_01_l [{shapeType=Capsule, length=18, radius=8, shapeName=dyn_under_arm_01_l_capsule, center=<10 6 0>}]
 dyn_under_arm_01_r [{shapeType=Capsule, length=18, radius=8, shapeName=dyn_under_arm_01_r_capsule, center=<-10 -6 0>}]
 dyn_under_arm_02_l [{shapeType=Capsule, length=18, radius=8, shapeName=dyn_under_arm_01_l_capsule, center=<10 6 0>}]
 dyn_under_arm_02_r [{shapeType=Capsule, length=18, radius=8, shapeName=dyn_back_01_r_capsule, center=<-10 -6 0>}]
 dyn_under_arm_03_l [{shapeType=Capsule, length=12, radius=1, shapeName=dyn_under_arm_03_l_capsule}]
 dyn_under_arm_03_r [{shapeType=Capsule, length=12, radius=1, shapeName=dyn_under_arm_03_r_capsule}]
 frnt_torso_01_l [{shapeType=Capsule, length=10, radius=7, shapeName=frnt_torso_01_l_capsule, center=<12 6 0>}]
 frnt_torso_01_r [{shapeType=Capsule, length=10, radius=7, shapeName=frnt_torso_01_r_capsule, center=<-12 -6 0>}]
 frnt_torso_02_l [{shapeType=Capsule, length=15, radius=7, shapeName=frnt_torso_02_r_capsule, center=<6 6 0>, rotation=<rot 90 0 0>}]
 frnt_torso_02_r [{shapeType=Capsule, length=15, radius=7, shapeName=frnt_torso_02_r_capsule, center=<-6 -6 0>, rotation=<rot 90 0 0>}]
 mid_back_01 [{shapeType=Capsule, length=8, radius=7, shapeName=frnt_torso_02_r_capsule, center=<8 -6 0>, rotation=<rot 90 0 0>}]
 mid_back_02 [{shapeType=Capsule, length=9, radius=7, shapeName=mid_back_02_capsule, center=<9 -6 0>, rotation=<rot 90 0 0>}]
 mstr_back_l [{shapeType=Capsule, length=10, radius=7, shapeName=mstr_back_l_capsule, center=<12 0 0>}]
 mstr_back_r [{shapeType=Capsule, length=10, radius=7, shapeName=mstr_back_r_capsule, center=<-12 0 0>}]
 mstr_frnt_torso_l [{shapeType=Capsule, length=10, radius=7, shapeName=mstr_frnt_torso_l_capsule, center=<12 0 0>}]
 mstr_frnt_torso_r [{shapeType=Capsule, length=10, radius=7, shapeName=mstr_frnt_torso_r_capsule, center=<-12 0 0>}]
 mstr_mid_back [{shapeType=Capsule, length=10, radius=7, shapeName=mstr_mid_back_capsule, center=<12 0 0>}]
 mstr_under_arm_l [{shapeType=Capsule, length=10, radius=7, shapeName=mstr_under_arm_l_capsule, center=<12 0 0>}]
 mstr_under_arm_r [{shapeType=Capsule, length=10, radius=7, shapeName=mstr_under_arm_r_capsule, center=<-12 0 0>}]
 pelvis [
  {shapeType=Capsule, length=10.25, radius=10.3125, center=<-2.6114 -2.8515 0>}
  {shapeType=Capsule, length=10.25, radius=11, shapeName=pelvis_capsule, center=<-12.7469 -1.285 0>}
 ]
 spine_01 [{shapeType=Capsule, length=10.5, radius=10, center=<-0.743 -3.4025 0>}]
 spine_02 [{shapeType=Capsule, length=10.5, radius=10.5, center=<-0.7431 -1.6978 0>}]
 spine_03 [{shapeType=Capsule, length=10.5, radius=10.5, center=<-0.743 -1.3306 0>}]
 spine_04 [{shapeType=Capsule, length=10, radius=8, center=<-0.7431 -1.5203 0.0001>}]
 thigh_l [{shapeType=Capsule, length=33, radius=11, center=<-18.1396 -0.8169 1.4515>, rotation=<rot 90 1.7899 -0.8059>}]
 thigh_r [{shapeType=Capsule, length=33, radius=11, center=<18.1396 0.8169 -1.4515>, rotation=<rot 90 1.7899 -0.8059>}]
 under_arm_01_l [{shapeType=Capsule, length=15, radius=7, shapeName=frnt_torso_02_r_capsule, center=<6 6 0>, rotation=<rot 90 0 0>}]
 under_arm_01_r [{shapeType=Capsule, length=15, radius=7, shapeName=frnt_torso_02_r_capsule, center=<-6 -6 0>, rotation=<rot 90 0 0>}]
 under_arm_02_l [{shapeType=Capsule, length=15, radius=7, shapeName=frnt_torso_02_r_capsule, center=<6 6 0>, rotation=<rot 90 0 0>}]
 under_arm_02_r [{shapeType=Capsule, length=15, radius=7, shapeName=frnt_torso_02_r_capsule, center=<-6 -6 0>, rotation=<rot 90 0 0>}]
}
constraints [
 {
  bone1Name=spine_01
  bone2Name=pelvis
  swing1LimitDegrees=45
  swing1Motion=Free
  swing2LimitDegrees=45
  swing2Motion=Free
  twistLimitDegrees=45
  twistMotion=Free
 }
 {
  bone1Name=spine_02
  bone2Name=spine_01
  swing1LimitDegrees=45
  swing1Motion=Free
  swing2LimitDegrees=45
  swing2Motion=Free
  twistLimitDegrees=45
  twistMotion=Free
 }
 {
  bone1Name=spine_03
  bone2Name=spine_02
  swing1LimitDegrees=45
  swing1Motion=Free
  swing2LimitDegrees=45
  swing2Motion=Free
  twistLimitDegrees=45
  twistMotion=Free
 }
 {
  bone1Name=spine_04
  bone2Name=spine_03
  swing1LimitDegrees=45
  swing1Motion=Free
  swing2LimitDegrees=45
  swing2Motion=Free
  twistLimitDegrees=45
  twistMotion=Free
 }
 {
  bone1Name=mstr_under_arm_l
  bone2Name=spine_04
  swing1LimitDegrees=45
  swing1Motion=Free
  swing2LimitDegrees=45
  swing2Motion=Free
  twistLimitDegrees=45
  twistMotion=Free
 }
 {bone1Name=under_arm_01_l, bone2Name=mstr_under_arm_l, swing1LimitDegrees=20, swing2LimitDegrees=20, twistLimitDegrees=5}
 {bone1Name=under_arm_02_l, bone2Name=under_arm_01_l, swing1LimitDegrees=20, swing2LimitDegrees=20, twistLimitDegrees=5}
 {
  bone1Name=dyn_mstr_under_arm_l
  bone2Name=under_arm_02_l
  swing1LimitDegrees=45
  swing1Motion=Locked
  swing2LimitDegrees=45
  swing2Motion=Locked
  twistLimitDegrees=45
  twistMotion=Locked
 }
 {bone1Name=dyn_under_arm_01_l, bone2Name=dyn_mstr_under_arm_l, swing1LimitDegrees=20, swing2LimitDegrees=20, twistLimitDegrees=5}
 {bone1Name=dyn_under_arm_02_l, bone2Name=dyn_under_arm_01_l, swing1LimitDegrees=20, swing2LimitDegrees=20, twistLimitDegrees=5}
 {
  bone1Name=dyn_under_arm_03_l
  bone2Name=dyn_under_arm_02_l
  swing1LimitDegrees=45
  swing1Motion=Locked
  swing2LimitDegrees=45
  swing2Motion=Locked
  twistLimitDegrees=2
  twistMotion=Locked
 }
 {
  bone1Name=mstr_mid_back
  bone2Name=spine_04
  swing1LimitDegrees=45
  swing1Motion=Free
  swing2LimitDegrees=45
  swing2Motion=Free
  twistLimitDegrees=45
  twistMotion=Free
 }
 {bone1Name=mid_back_01, bone2Name=mstr_mid_back, swing1LimitDegrees=12, swing2LimitDegrees=15, twistLimitDegrees=5}
 {bone1Name=mid_back_02, bone2Name=mid_back_01, swing1LimitDegrees=15, swing2LimitDegrees=15, twistLimitDegrees=5}
 {
  bone1Name=dyn_mstr_mid_back
  bone2Name=mid_back_02
  swing1LimitDegrees=45
  swing1Motion=Locked
  swing2LimitDegrees=45
  swing2Motion=Locked
  twistLimitDegrees=45
  twistMotion=Locked
 }
 {bone1Name=dyn_mid_back_01, bone2Name=dyn_mstr_mid_back, swing1LimitDegrees=20, swing2LimitDegrees=22, twistLimitDegrees=5}
 {bone1Name=dyn_mid_back_02, bone2Name=dyn_mid_back_01, swing1LimitDegrees=22, swing2LimitDegrees=22, twistLimitDegrees=5}
 {bone1Name=dyn_mid_back_03, bone2Name=dyn_mid_back_02, swing1LimitDegrees=13, swing2LimitDegrees=13, twistLimitDegrees=5}
 {
  bone1Name=mstr_frnt_torso_l
  bone2Name=spine_04
  swing1LimitDegrees=45
  swing1Motion=Free
  swing2LimitDegrees=45
  swing2Motion=Free
  twistLimitDegrees=45
  twistMotion=Free
 }
 {bone1Name=frnt_torso_01_l, bone2Name=mstr_frnt_torso_l, swing1LimitDegrees=45, swing2LimitDegrees=45, twistLimitDegrees=2}
 {bone1Name=frnt_torso_02_l, bone2Name=frnt_torso_01_l, swing1LimitDegrees=45, swing2LimitDegrees=45, twistLimitDegrees=10}
 {
  bone1Name=dyn_mstr_frnt_torso_l
  bone2Name=frnt_torso_02_l
  swing1LimitDegrees=45
  swing1Motion=Locked
  swing2LimitDegrees=45
  swing2Motion=Locked
  twistLimitDegrees=45
  twistMotion=Locked
 }
 {bone1Name=dyn_frnt_torso_01_l, bone2Name=dyn_mstr_frnt_torso_l, swing1LimitDegrees=35, swing2LimitDegrees=35, twistLimitDegrees=10}
 {bone1Name=dyn_frnt_torso_02_l, bone2Name=dyn_frnt_torso_01_l, swing1LimitDegrees=35, swing2LimitDegrees=35, twistLimitDegrees=10}
 {
  bone1Name=dyn_frnt_torso_03_l
  bone2Name=dyn_frnt_torso_02_l
  swing1LimitDegrees=45
  swing1Motion=Locked
  swing2LimitDegrees=45
  swing2Motion=Locked
  twistLimitDegrees=2
  twistMotion=Locked
 }
 {
  bone1Name=mstr_back_l
  bone2Name=spine_04
  swing1LimitDegrees=45
  swing1Motion=Free
  swing2LimitDegrees=45
  swing2Motion=Free
  twistLimitDegrees=45
  twistMotion=Free
 }
 {bone1Name=back_01_l, bone2Name=mstr_back_l, swing1LimitDegrees=12, swing2LimitDegrees=15, twistLimitDegrees=5}
 {bone1Name=back_02_l, bone2Name=back_01_l, swing1LimitDegrees=15, swing2LimitDegrees=15, twistLimitDegrees=5}
 {
  bone1Name=dyn_mstr_back_l
  bone2Name=back_02_l
  swing1LimitDegrees=45
  swing1Motion=Locked
  swing2LimitDegrees=45
  swing2Motion=Locked
  twistLimitDegrees=45
  twistMotion=Locked
 }
 {bone1Name=dyn_back_01_l, bone2Name=dyn_mstr_back_l, swing1LimitDegrees=20, swing2LimitDegrees=22, twistLimitDegrees=5}
 {bone1Name=dyn_back_02_l, bone2Name=dyn_back_01_l, swing1LimitDegrees=22, swing2LimitDegrees=22, twistLimitDegrees=5}
 {bone1Name=dyn_back_03_l, bone2Name=dyn_back_02_l, swing1LimitDegrees=13, swing2LimitDegrees=13, twistLimitDegrees=5}
 {
  bone1Name=mstr_back_r
  bone2Name=spine_04
  swing1LimitDegrees=45
  swing1Motion=Free
  swing2LimitDegrees=45
  swing2Motion=Free
  twistLimitDegrees=45
  twistMotion=Free
 }
 {bone1Name=back_01_r, bone2Name=mstr_back_r, swing1LimitDegrees=12, swing2LimitDegrees=15, twistLimitDegrees=5}
 {bone1Name=back_02_r, bone2Name=back_01_r, swing1LimitDegrees=15, swing2LimitDegrees=15, twistLimitDegrees=5}
 {
  bone1Name=dyn_mstr_back_r
  bone2Name=back_02_r
  swing1LimitDegrees=45
  swing1Motion=Locked
  swing2LimitDegrees=45
  swing2Motion=Locked
  twistLimitDegrees=45
  twistMotion=Locked
 }
 {bone1Name=dyn_back_01_r, bone2Name=dyn_mstr_back_r, swing1LimitDegrees=20, swing2LimitDegrees=22, twistLimitDegrees=5}
 {bone1Name=dyn_back_02_r, bone2Name=dyn_back_01_r, swing1LimitDegrees=22, swing2LimitDegrees=22, twistLimitDegrees=5}
 {bone1Name=dyn_back_03_r, bone2Name=dyn_back_02_r, swing1LimitDegrees=13, swing2LimitDegrees=13, twistLimitDegrees=5}
 {
  bone1Name=mstr_under_arm_r
  bone2Name=spine_04
  swing1LimitDegrees=45
  swing1Motion=Free
  swing2LimitDegrees=45
  swing2Motion=Free
  twistLimitDegrees=45
  twistMotion=Free
 }
 {bone1Name=under_arm_01_r, bone2Name=mstr_under_arm_r, swing1LimitDegrees=20, swing2LimitDegrees=20, twistLimitDegrees=5}
 {bone1Name=under_arm_02_r, bone2Name=under_arm_01_r, swing1LimitDegrees=20, swing2LimitDegrees=20, twistLimitDegrees=5}
 {
  bone1Name=dyn_mstr_under_arm_r
  bone2Name=under_arm_02_r
  swing1LimitDegrees=45
  swing1Motion=Locked
  swing2LimitDegrees=45
  swing2Motion=Locked
  twistLimitDegrees=45
  twistMotion=Locked
 }
 {bone1Name=dyn_under_arm_01_r, bone2Name=dyn_mstr_under_arm_r, swing1LimitDegrees=20, swing2LimitDegrees=20, twistLimitDegrees=5}
 {bone1Name=dyn_under_arm_02_r, bone2Name=dyn_under_arm_01_r, swing1LimitDegrees=20, swing2LimitDegrees=20, twistLimitDegrees=5}
 {
  bone1Name=dyn_under_arm_03_r
  bone2Name=dyn_under_arm_02_r
  swing1LimitDegrees=45
  swing1Motion=Locked
  swing2LimitDegrees=45
  swing2Motion=Locked
  twistLimitDegrees=2
  twistMotion=Locked
 }
 {
  bone1Name=mstr_frnt_torso_r
  bone2Name=spine_04
  swing1LimitDegrees=45
  swing1Motion=Free
  swing2LimitDegrees=45
  swing2Motion=Free
  twistLimitDegrees=45
  twistMotion=Free
 }
 {bone1Name=frnt_torso_01_r, bone2Name=mstr_frnt_torso_r, swing1LimitDegrees=45, swing2LimitDegrees=45, twistLimitDegrees=2}
 {bone1Name=frnt_torso_02_r, bone2Name=frnt_torso_01_r, swing1LimitDegrees=35, swing2LimitDegrees=35, twistLimitDegrees=10}
 {
  bone1Name=dyn_mstr_frnt_torso_r
  bone2Name=frnt_torso_02_r
  swing1LimitDegrees=25
  swing1Motion=Locked
  swing2LimitDegrees=25
  swing2Motion=Locked
  twistLimitDegrees=4
  twistMotion=Locked
 }
 {bone1Name=dyn_frnt_torso_01_r, bone2Name=dyn_mstr_frnt_torso_r, swing1LimitDegrees=35, swing2LimitDegrees=35, twistLimitDegrees=10}
 {bone1Name=dyn_frnt_torso_02_r, bone2Name=dyn_frnt_torso_01_r, swing1LimitDegrees=35, swing2LimitDegrees=35, twistLimitDegrees=10}
 {
  bone1Name=dyn_frnt_torso_03_r
  bone2Name=dyn_frnt_torso_02_r
  swing1LimitDegrees=45
  swing1Motion=Locked
  swing2LimitDegrees=45
  swing2Motion=Locked
  twistLimitDegrees=2
  twistMotion=Locked
 }
 {
  bone1Name=dyn_mid_back_02
  bone2Name=dyn_back_02_r
  swing1LimitDegrees=45
  swing1Motion=Free
  swing2LimitDegrees=45
  swing2Motion=Free
  twistLimitDegrees=45
  twistMotion=Free
 }
 {
  bone1Name=dyn_mid_back_02
  bone2Name=dyn_back_02_l
  swing1LimitDegrees=45
  swing1Motion=Free
  swing2LimitDegrees=45
  swing2Motion=Free
  twistLimitDegrees=45
  twistMotion=Free
 }
 {
  bone1Name=dyn_frnt_torso_02_l
  bone2Name=dyn_under_arm_02_l
  swing1LimitDegrees=45
  swing1Motion=Free
  swing2LimitDegrees=45
  swing2Motion=Free
  twistLimitDegrees=45
  twistMotion=Free
 }
 {
  bone1Name=dyn_frnt_torso_02_r
  bone2Name=dyn_under_arm_02_r
  swing1LimitDegrees=45
  swing1Motion=Free
  swing2LimitDegrees=45
  swing2Motion=Free
  twistLimitDegrees=45
  twistMotion=Free
 }
 {bone1Name=frnt_torso_01_r, bone2Name=under_arm_01_r, swing1LimitDegrees=10, swing2LimitDegrees=10, twistLimitDegrees=10}
 {bone1Name=frnt_torso_01_l, bone2Name=under_arm_01_l, swing1LimitDegrees=10, swing2LimitDegrees=10, twistLimitDegrees=10}
]
