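# Physics asset for the left arm mesh of the female large purple leather jacket cosmetic, defining physics for its arm strap.
# A small master capsule anchors a chain of three identically sized capsule bodies on the dynamic strap bones.
# Constraints link each bone to the previous one with tight twist and moderate swing limits, with one wider swing axis at the root joint.
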PhysicsAsset /BRCosmetics/Characters/Player/Female/Large/Bodies/F_LRG_Leather_Jacket_Purple/Meshes/F_LRG_Leather_Jacket_Purple_LeftArm_Physics
bodies {
 dyn_coat_arm_strap_01_l [{
   shapeType=Capsule
   length=5.7887
   radius=1.7306
   shapeName=dyn_coat_arm_strap_mstr_l_capsule
   center=<5.5037 0 0.0005>
   rotation=<rot 90 0.844 0.8441>
  }]
 dyn_coat_arm_strap_02_l [{
   shapeType=Capsule
   length=5.7887
   radius=1.7306
   shapeName=dyn_coat_arm_strap_02_l_capsule
   center=<5.5037 0 0.0005>
   rotation=<rot 90 1.2659 1.2659>
  }]
 dyn_coat_arm_strap_03_l [{
   shapeType=Capsule
   length=5.7887
   radius=1.7306
   shapeName=dyn_coat_arm_strap_03_l_capsule
   center=<5.5037 0 0.0005>
   rotation=<rot 90 0.0099 0.0099>
  }]
 mstr_coat_arm_strap_l [{shapeType=Capsule, length=1.0666, radius=0.5, shapeName=mstr_coat_arm_strap_l_capsule}]
}
constraints [
 {bone1Name=dyn_coat_arm_strap_03_l, bone2Name=dyn_coat_arm_strap_02_l, swing1LimitDegrees=40, swing2LimitDegrees=40, twistLimitDegrees=5}
 {bone1Name=dyn_coat_arm_strap_02_l, bone2Name=dyn_coat_arm_strap_01_l, swing1LimitDegrees=40, swing2LimitDegrees=40, twistLimitDegrees=5}
 {bone1Name=dyn_coat_arm_strap_01_l, bone2Name=mstr_coat_arm_strap_l, swing1LimitDegrees=60, swing2LimitDegrees=40, twistLimitDegrees=5}
]
