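# Physics asset for the right-arm strap on the large female purple leather jacket body cosmetic.
# Small capsule bodies cover a master strap bone and a chain of three dynamic strap bones.
# Angular-limited constraints link each segment to the previous one back to the master, with tight twist and a wider first swing at the root joint, shaping how the strap can dangle.
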PhysicsAsset /BRCosmetics/Characters/Player/Female/Large/Bodies/F_LRG_Leather_Jacket_Purple/Meshes/F_LRG_Leather_Jacket_Purple_RightArm_Physics
bodies {
 dyn_coat_arm_strap_01_r [{
   shapeType=Capsule
   length=5.4311
   radius=1.5392
   shapeName=dyn_coat_arm_strap_01_r_capsule
   center=<4.439 0 0.0005>
   rotation=<rot 90 -1.8984 -1.8983>
  }]
 dyn_coat_arm_strap_02_r [{
   shapeType=Capsule
   length=5.4311
   radius=1.5392
   shapeName=dyn_coat_arm_strap_02_r_capsule
   center=<4.439 0 0.0005>
   rotation=<rot 90 -1.8984 -1.8983>
  }]
 dyn_coat_arm_strap_03_r [{
   shapeType=Capsule
   length=5.4311
   radius=1.5392
   shapeName=dyn_coat_arm_strap_03_r_capsule
   center=<4.439 0 0.0005>
   rotation=<rot 90 -1.2659 -1.2659>
  }]
 mstr_coat_arm_strap_r [{shapeType=Capsule, length=0, radius=0.5, shapeName=mstr_coat_arm_strap_r_capsule}]
}
constraints [
 {bone1Name=dyn_coat_arm_strap_03_r, bone2Name=dyn_coat_arm_strap_02_r, swing1LimitDegrees=40, swing2LimitDegrees=40, twistLimitDegrees=5}
 {bone1Name=dyn_coat_arm_strap_02_r, bone2Name=dyn_coat_arm_strap_01_r, swing1LimitDegrees=40, swing2LimitDegrees=40, twistLimitDegrees=5}
 {bone1Name=dyn_coat_arm_strap_01_r, bone2Name=mstr_coat_arm_strap_r, swing1LimitDegrees=60, swing2LimitDegrees=40, twistLimitDegrees=5}
]
