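# Control Rig for the torso of the large female purple leather jacket, applying corrective offsets to jacket bones as the body moves.
# Spherical pose readers on clavicles and upper arms weight a shared function that adds local transform offsets to underarm, front, back and armpit bones.
# The mid-back master is constrained between both back masters, and is_crouching triggers fixed right-front offsets.
# Thigh-reader offsets, spine-bend correctives and control-to-arm rotation setters are unwired from execution.
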ControlRigBlueprint /BRCosmetics/Characters/Player/Female/Large/Bodies/F_LRG_Leather_Jacket_Purple/Meshes/F_LRG_Leather_Jacket_Purple_Torso_CtrlRig
graphs [
 ContainedGraph {nodes [
   n1 = RigVMFunctionEntryNode()
   RigUnit_GetTransform = &RigUnit_GetTransform(Space=LocalSpace, bInitial=false, Item=@n1.Joint)
   n2 = &RigVMFunction_MathQuaternionToEuler(RotationOrder=ZYX, Value=@RigUnit_GetTransform.Transform.Rotation)
   > RigUnit_SetTransform = &RigUnit_SetTransform(
    Space=LocalSpace
    Value=(Rotation=(X=0.0132,Y=0.4099,Z=-0.0222,W=0.9118),Translation=(X=23.6045,Y=-6.37,Z=140.1115),Scale3D=(X=1,Y=1,Z=1))
    bInitial=false
    Item=@n1.Joint
    Weight=@n1.Weight
    Value.Translation.Z=&RigVMFunction_MathFloatAdd(B=@n1.Translate Z, A=@RigUnit_GetTransform.Transform.Translation.Z) {template="Add::Execute(in A,in B,out Result)"}.Result
    Value.Scale3D.Z=&RigVMFunction_MathFloatAdd(B=@n1.Scale Z, A=@RigUnit_GetTransform.Transform.Scale3D.Z) {template="Add::Execute(in A,in B,out Result)"}.Result
    Value.Translation.Y=&RigVMFunction_MathFloatAdd(B=@n1.Translate Y, A=@RigUnit_GetTransform.Transform.Translation.Y) {template="Add::Execute(in A,in B,out Result)"}.Result
    Value.Scale3D.Y=&RigVMFunction_MathFloatAdd(B=@n1.Scale Y, A=@RigUnit_GetTransform.Transform.Scale3D.Y) {template="Add::Execute(in A,in B,out Result)"}.Result
    Value.Translation.X=&RigVMFunction_MathFloatAdd(B=@n1.Translate X, A=@RigUnit_GetTransform.Transform.Translation.X) {template="Add::Execute(in A,in B,out Result)"}.Result
    Value.Scale3D.X=&RigVMFunction_MathFloatAdd(B=@n1.Scale X, A=@RigUnit_GetTransform.Transform.Scale3D.X) {template="Add::Execute(in A,in B,out Result)"}.Result
    Value.Rotation=&RigVMFunction_MathQuaternionFromEuler(
     Euler=(X=0,Y=0,Z=0)
     Euler.X=&RigVMFunction_MathFloatAdd(B=@n1.Rotate X, A=@n2.Result.X) {template="Add::Execute(in A,in B,out Result)"}.Result
     Euler.Y=&RigVMFunction_MathFloatAdd(B=@n1.Rotate Y, A=@n2.Result.Y) {template="Add::Execute(in A,in B,out Result)"}.Result
     Euler.Z=&RigVMFunction_MathFloatAdd(B=@n1.Rotate Z, A=@n2.Result.Z) {template="Add::Execute(in A,in B,out Result)"}.Result
    ).Result
   )
   > RigVMFunctionReturnNode()
  ]}
 RigVMFunctionLibrary {nodes [Set Transforms for Bone = RigVMCollapseNode(
    ExecuteContext=()
    Joint=()
    Rotate X=0
    Rotate Y=0
    Rotate Z=0
    Scale X=0
    Scale Y=0
    Scale Z=0
    Translate X=0
    Translate Y=0
    Translate Z=0
    Weight=1
   )]}
 RigVMModel {nodes [
   RigUnit_BeginExecution = &RigUnit_BeginExecution()
   n1 = /Script/ControlRig.RigUnit_SequenceExecution(ExecuteContext=@RigUnit_BeginExecution.ExecutePin) {is=struct RigVMUnitNode}
   n2 = &RigUnit_SphericalPoseReader(
    ActiveRegionScaleFactors=(PositiveWidth=2,NegativeWidth=2,PositiveHeight=1,NegativeHeight=3)
    ActiveRegionSize=0.1
    Debug=(bDrawDebug=false,DebugScale=25,DebugSegments=20,DebugThickness=0.25)
    DriverAxis=(X=1,Y=0,Z=0)
    DriverItem=(Type=Bone,Name="clavicle_l")
    FalloffRegionScaleFactors=(PositiveWidth=0.5,NegativeWidth=0.2,PositiveHeight=0.5,NegativeHeight=1)
    FalloffSize=0.5
    OptionalParentItem=(Type=Bone)
    RotationOffset=(X=150,Y=35,Z=0)
    ExecutePin=@n1.A
   )
   > Set Transforms for Bone_1_1_1_1_1_1_2_5 = RigVMFunctionReferenceNode(
    Joint=(Type=Bone,Name="mstr_under_arm_l")
    Rotate X=0
    Rotate Y=0
    Rotate Z=0
    Scale X=0
    Scale Y=0
    Scale Z=0
    Translate X=2
    Translate Y=1.5
    Translate Z=-1
    Weight=@n2.OutputParam
   )
   > Set Transforms for Bone_1_1_1_1_1_1_2_1_4 = RigVMFunctionReferenceNode(
    Joint=(Type=Bone,Name="mstr_frnt_torso_l")
    Rotate X=0
    Rotate Y=0
    Rotate Z=0
    Scale X=0
    Scale Y=0
    Scale Z=0
    Translate X=2.5
    Translate Y=0
    Translate Z=0
    Weight=@n2.OutputParam
   )
   > Set Transforms for Bone_1_1_1_1_1_1_2_1_1_3 = RigVMFunctionReferenceNode(
    Joint=(Type=Bone,Name="mstr_back_l")
    Rotate X=0
    Rotate Y=0
    Rotate Z=0
    Scale X=0
    Scale Y=0
    Scale Z=0
    Translate X=2.2
    Translate Y=1
    Translate Z=-1
    Weight=@n2.OutputParam
   )
   > n3 = &RigUnit_SphericalPoseReader(
    ActiveRegionScaleFactors=(PositiveWidth=4,NegativeWidth=4,PositiveHeight=1.5,NegativeHeight=1)
    ActiveRegionSize=0.1
    Debug=(bDrawDebug=false,DebugScale=25,DebugSegments=20,DebugThickness=0.25)
    DriverAxis=(X=1,Y=0,Z=0)
    DriverItem=(Type=Bone,Name="upperarm_l")
    FalloffRegionScaleFactors=(PositiveWidth=1,NegativeWidth=1,PositiveHeight=0.3,NegativeHeight=0.1)
    FalloffSize=0.6
    OptionalParentItem=(Type=Bone)
    RotationOffset=(X=75,Y=0,Z=-30)
   )
   > Set Transforms for Bone_1_1_1_1_1_1_2_4_1_3 = RigVMFunctionReferenceNode(
    Joint=(Type=Bone,Name="armpit_01_l")
    Rotate X=0
    Rotate Y=0
    Rotate Z=0
    Scale X=0
    Scale Y=0
    Scale Z=0
    Translate X=-5
    Translate Y=0
    Translate Z=2.5
    Weight=@n3.OutputParam
   )
   n4 = &RigUnit_SequenceExecution(ExecuteContext=@n1.D)
   &RigUnit_ParentConstraint(
    AdvancedSettings=(RotationOrderForFilter=XZY)
    Child=(Type=Bone,Name="mstr_mid_back")
    Parents=((Item=(Type=Bone,Name="mstr_back_l"),Weight=1),(Item=(Type=Bone,Name="mstr_back_r"),Weight=1))
    Weight=1
    ExecutePin=@n4.A
   )
   Branch_1_RigVMUnitNode = &RigVMFunction_ControlFlowBranch(ExecuteContext=@n4.B, Condition=RigVMVariableNode() {variable=is_crouching}.Value)
   n5 = &RigUnit_SequenceExecution(ExecuteContext=@Branch_1_RigVMUnitNode.True)
   Set Transforms for Bone_1_1_1_1_1_1_3_1 = RigVMFunctionReferenceNode(
    Joint=(Type=Bone,Name="frnt_torso_01_r")
    Rotate X=35
    Rotate Y=0
    Rotate Z=0
    Scale X=0
    Scale Y=0
    Scale Z=0
    Translate X=3
    Translate Y=1.5
    Translate Z=-1.5
    Weight=1
    ExecuteContext=@n5.A
   )
   Set Transforms for Bone_1_1_1_1_1_1_2_1_3_1_1_2 = RigVMFunctionReferenceNode(
    Joint=(Type=Bone,Name="frnt_torso_01_r")
    Rotate X=0
    Rotate Y=0
    Rotate Z=0
    Scale X=0
    Scale Y=0
    Scale Z=0
    Translate X=1
    Translate Y=0
    Translate Z=0
    Weight=1
    ExecuteContext=@n5.B
   )
   > Set Transforms for Bone_1_1_1_1_1_1_2_1_3_1_1_1_1 = RigVMFunctionReferenceNode(
    Joint=(Type=Bone,Name="frnt_torso_02_r")
    Rotate X=0
    Rotate Y=0
    Rotate Z=0
    Scale X=0
    Scale Y=0
    Scale Z=0
    Translate X=1
    Translate Y=0
    Translate Z=0
    Weight=1
   )
   n6 = &RigUnit_SphericalPoseReader(
    ActiveRegionScaleFactors=(PositiveWidth=2,NegativeWidth=2,PositiveHeight=3,NegativeHeight=1)
    ActiveRegionSize=0.1
    Debug=(bDrawDebug=false,bDraw2D=false,bDrawLocalAxes=false,DebugScale=25,DebugSegments=20,DebugThickness=0.25)
    DriverAxis=(X=-1,Y=0,Z=0)
    DriverItem=(Type=Bone,Name="clavicle_r")
    FalloffRegionScaleFactors=(PositiveWidth=0.5,NegativeWidth=0.2,PositiveHeight=1,NegativeHeight=0.5)
    FalloffSize=0.5
    FlipHeightScaling=false
    FlipWidthScaling=false
    OptionalParentItem=(Type=Bone)
    RotationOffset=(X=35,Y=-35,Z=0)
    ExecutePin=@n1.B
   )
   > Set Transforms for Bone_1_1_1_1_1_1_2_3_1 = RigVMFunctionReferenceNode(
    Joint=(Type=Bone,Name="mstr_under_arm_r")
    Rotate X=0
    Rotate Y=0
    Rotate Z=0
    Scale X=0
    Scale Y=0
    Scale Z=0
    Translate X=2
    Translate Y=1.5
    Translate Z=1
    Weight=@n6.OutputParam
   )
   > Set Transforms for Bone_1_1_1_1_1_1_2_1_3_1 = RigVMFunctionReferenceNode(
    Joint=(Type=Bone,Name="mstr_frnt_torso_r")
    Rotate X=0
    Rotate Y=0
    Rotate Z=0
    Scale X=0
    Scale Y=0
    Scale Z=0
    Translate X=2.5
    Translate Y=0
    Translate Z=0
    Weight=@n6.OutputParam
   )
   > Set Transforms for Bone_1_1_1_1_1_1_2_1_1_2_1 = RigVMFunctionReferenceNode(
    Joint=(Type=Bone,Name="mstr_back_r")
    Rotate X=0
    Rotate Y=0
    Rotate Z=0
    Scale X=0
    Scale Y=0
    Scale Z=0
    Translate X=2.2
    Translate Y=1
    Translate Z=-1
    Weight=@n6.OutputParam
   )
   > n7 = &RigUnit_SphericalPoseReader(
    ActiveRegionScaleFactors=(PositiveWidth=4,NegativeWidth=4,PositiveHeight=1.5,NegativeHeight=1)
    ActiveRegionSize=0.1
    Debug=(bDrawDebug=false,DebugScale=25,DebugSegments=20,DebugThickness=0.25)
    DriverAxis=(X=-1,Y=0,Z=0)
    DriverItem=(Type=Bone,Name="upperarm_r")
    FalloffRegionScaleFactors=(PositiveWidth=1,NegativeWidth=0.5,PositiveHeight=0.3,NegativeHeight=0.1)
    FalloffSize=0.6
    FlipHeightScaling=false
    FlipWidthScaling=false
    OptionalParentItem=(Type=Bone)
    RotationOffset=(X=-105,Y=0,Z=-30)
   )
   > Set Transforms for Bone_1_1_1_1_1_1_2_4_1_2_1 = RigVMFunctionReferenceNode(
    Joint=(Type=Bone,Name="armpit_01_r")
    Rotate X=0
    Rotate Y=0
    Rotate Z=0
    Scale X=0
    Scale Y=0
    Scale Z=0
    Translate X=5
    Translate Y=0
    Translate Z=-2.5
    Weight=@n7.OutputParam
   )
   n8 = &RigUnit_SphericalPoseReader(
    ActiveRegionScaleFactors=(PositiveWidth=3,NegativeWidth=1,PositiveHeight=2,NegativeHeight=3)
    ActiveRegionSize=0.1
    Debug=(bDrawDebug=false,bDrawLocalAxes=false,DebugScale=25,DebugSegments=20,DebugThickness=0.25)
    DriverAxis=(X=1,Y=0,Z=0)
    DriverItem=(Type=Bone,Name="thigh_r")
    FalloffRegionScaleFactors=(PositiveWidth=1,NegativeWidth=0.3,PositiveHeight=1,NegativeHeight=1)
    FalloffSize=0.4
    OptionalParentItem=(Type=Bone)
    RotationOffset=(X=75,Y=0,Z=50)
    ExecutePin=@n1.C
   )
   Set Transforms for Bone = RigVMFunctionReferenceNode(
    ExecuteContext=()
    Joint=(Type=Bone,Name="dyn_frnt_torso_01_r")
    Rotate X=2.5
    Rotate Y=-2.5
    Rotate Z=2.5
    Scale X=0
    Scale Y=0
    Scale Z=0
    Translate X=0
    Translate Y=0
    Translate Z=0
    Weight=@n8.OutputParam
   )
   > Set Transforms for Bone_1 = RigVMFunctionReferenceNode(
    Joint=(Type=Bone,Name="dyn_frnt_torso_02_r")
    Rotate X=0
    Rotate Y=-10
    Rotate Z=5
    Scale X=0
    Scale Y=0
    Scale Z=0
    Translate X=0
    Translate Y=0
    Translate Z=0
    Weight=@n8.OutputParam
   )
   > Set Transforms for Bone_2 = RigVMFunctionReferenceNode(
    Joint=()
    Rotate X=0
    Rotate Y=-5
    Rotate Z=-5
    Scale X=0
    Scale Y=0
    Scale Z=0
    Translate X=0
    Translate Y=0
    Translate Z=0
    Weight=@n8.OutputParam
   )
   n9 = &RigVMFunction_MathQuaternionToEuler(
    RotationOrder=ZYX
    Value=&RigUnit_GetRelativeTransformForItem(Child=(Type=Bone,Name="spine_05"), Parent=(Type=Bone,Name="pelvis")).RelativeTransform.Rotation
   )
   n10 = &RigVMFunction_MathFloatRemap(SourceMaximum=30, SourceMinimum=6, TargetMinimum=0, bClamp=true, Value=@n9.Result.Z)
   Set Transforms for Bone_1_1_1_1_1_1_2_1_3_1_1 = RigVMFunctionReferenceNode(
    ExecuteContext=()
    Joint=(Type=Bone,Name="frnt_torso_01_l")
    Rotate X=0
    Rotate Y=0
    Rotate Z=0
    Scale X=0
    Scale Y=0
    Scale Z=0
    Translate X=-1
    Translate Y=0
    Translate Z=0
    Weight=@n10.Result
   )
   > Set Transforms for Bone_1_1_1_1_1_1_2_1_3_1_1_1 = RigVMFunctionReferenceNode(
    Joint=(Type=Bone,Name="frnt_torso_02_l")
    Rotate X=0
    Rotate Y=0
    Rotate Z=0
    Scale X=0
    Scale Y=0
    Scale Z=0
    Translate X=-1
    Translate Y=0
    Translate Z=0
    Weight=@n10.Result
   )
   n11 = &RigVMFunction_MathFloatRemap(SourceMaximum=40, SourceMinimum=10, TargetMinimum=0, bClamp=true, Value=@n9.Result.X)
   Set Transforms for Bone_1_1_1_1_1_1_2_1_3_1_1_3 = RigVMFunctionReferenceNode(
    ExecuteContext=()
    Joint=(Type=Bone,Name="dyn_mstr_frnt_torso_l")
    Rotate X=0
    Rotate Y=0
    Rotate Z=0
    Scale X=0
    Scale Y=0
    Scale Z=0
    Translate X=-10
    Translate Y=0
    Translate Z=0
    Weight=@n11.Result
   )
   > Set Transforms for Bone_1_1_1_1_1_1_2_1_3_1_1_2_1 = RigVMFunctionReferenceNode(
    Joint=(Type=Bone,Name="frnt_torso_01_r")
    Rotate X=0
    Rotate Y=0
    Rotate Z=0
    Scale X=0
    Scale Y=0
    Scale Z=0
    Translate X=0
    Translate Y=0
    Translate Z=0
    Weight=@n11.Result
   )
   &RigUnit_SetRotation(
    ExecutePin=()
    Item=(Type=Bone,Name="clavicle_r")
    Space=GlobalSpace
    Weight=1
    Value=&RigUnit_GetTransform(Item=(Type=Control,Name="clavicle_r_ctrl")).Transform.Rotation
   )
   > &RigUnit_SetRotation(
    Item=(Type=Bone,Name="upperarm_r")
    Space=GlobalSpace
    Weight=1
    Value=&RigUnit_GetTransform(Item=(Type=Control,Name="upperarm_r_ctrl")).Transform.Rotation
   )
   > &RigUnit_SetRotation(
    Item=(Type=Bone,Name="clavicle_l")
    Space=GlobalSpace
    Weight=1
    Value=&RigUnit_GetTransform(Item=(Type=Control,Name="clavicle_l_ctrl")).Transform.Rotation
   )
   > &RigUnit_SetRotation(
    Item=(Type=Bone,Name="upperarm_l")
    Space=GlobalSpace
    Weight=1
    Value=&RigUnit_GetTransform(Item=(Type=Control,Name="upperarm_l_ctrl")).Transform.Rotation
   )
   Set Transforms for Bone_1_1_1_1_1_1_1_1 = RigVMFunctionReferenceNode(
    ExecuteContext=()
    Joint=(Type=Bone,Name="frnt_torso_02_l")
    Rotate X=35
    Rotate Y=0
    Rotate Z=0
    Scale X=0
    Scale Y=0
    Scale Z=0
    Translate X=-2
    Translate Y=0
    Translate Z=0
    Weight=1
   )
   Set Transforms for Bone_1_1_1_1_1_1_2_1_3_1_1_1_1_1 = RigVMFunctionReferenceNode(
    ExecuteContext=()
    Joint=(Type=Bone,Name="frnt_torso_02_r")
    Rotate X=0
    Rotate Y=0
    Rotate Z=0
    Scale X=0
    Scale Y=0
    Scale Z=0
    Translate X=1
    Translate Y=0
    Translate Z=0
    Weight=1
   )
   Set Transforms for Bone_1_1_1_1_1_1_2_1_3_1_1_1_2 = RigVMFunctionReferenceNode(
    ExecuteContext=()
    Joint=(Type=Bone,Name="frnt_torso_02_l")
    Rotate X=0
    Rotate Y=0
    Rotate Z=0
    Scale X=0
    Scale Y=0
    Scale Z=0
    Translate X=-1
    Translate Y=0
    Translate Z=0
    Weight=1
   )
   Set Transforms for Bone_1_1_1_1_1_1_3 = RigVMFunctionReferenceNode(
    ExecuteContext=()
    Joint=(Type=Bone,Name="frnt_torso_01_l")
    Rotate X=-70
    Rotate Y=0
    Rotate Z=45
    Scale X=0
    Scale Y=0
    Scale Z=0
    Translate X=-3
    Translate Y=-4
    Translate Z=6
    Weight=1
   )
   Set Transforms for Bone_3 = RigVMFunctionReferenceNode(
    ExecuteContext=()
    Joint=(Type=Bone,Name="under_arm_01_l")
    Rotate X=0
    Rotate Y=-10
    Rotate Z=5
    Scale X=0
    Scale Y=0
    Scale Z=0
    Translate X=-1
    Translate Y=0
    Translate Z=0
    Weight=1
   )
   Set Transforms for Bone_3_1 = RigVMFunctionReferenceNode(
    ExecuteContext=()
    Joint=(Type=Bone,Name="under_arm_01_r")
    Rotate X=0
    Rotate Y=13
    Rotate Z=-4
    Scale X=0
    Scale Y=0
    Scale Z=0
    Translate X=1.5
    Translate Y=0
    Translate Z=0
    Weight=1
   )
  ]}
]
hierarchy [
 root {loc=<0 0 0>, rot=<0 0 0 1>}
 pelvis {parent=root, loc=<0 -0.648 94.7699>, rot=<0.0127 0.707 -0.0127 -0.707>}
 spine_01 {parent=pelvis, loc=<5.0975 -0.6388 0>, rot=<0 0 0.034 -0.9994>}
 spine_02 {parent=spine_01, loc=<7.2518 -0.035 0>, rot=<0 0 0.0238 -0.9997>}
 spine_03 {parent=spine_02, loc=<6.0104 0 0>, rot=<0 0 0.0149 -0.9999>}
 spine_04 {parent=spine_03, loc=<8.1024 0 0>, rot=<0 0 -0.0132 -0.9999>}
 spine_05 {parent=spine_04, loc=<10.0146 0 0>, rot=<0 0 -0.0015 -1>}
 clavicle_l {parent=spine_05, loc=<4.1651 -0.3023 -5.0942>, rot=<0.0465 -0.7289 -0.009 -0.6829>}
 upperarm_l {parent=clavicle_l, loc=<11.2117 -1.1486 -0.0332>, rot=<0.0179 -0.2982 -0.028 -0.9539>}
 lowerarm_l {parent=upperarm_l, loc=<26.5013 -0.0835 0.2605>, rot=<0.0036 0.0161 -0.2298 -0.9731>}
 hand_l {parent=lowerarm_l, loc=<23.1356 -1.1606 -0.2612>, rot=<-0.6704 -0.0028 0.0567 -0.7398>}
 index_metacarpal_l {parent=hand_l, loc=<1.6535 0.0183 -2.96>, rot=<0.0158 -0.0124 -0.0173 -0.9996>}
 index_01_l {parent=index_metacarpal_l, loc=<8.3008 0.0357 -0.0524>, rot=<-0.0044 0.0485 0.0924 -0.9945>}
 index_02_l {parent=index_01_l, loc=<3.3587 -0.3004 0.0452>, rot=<-0.002 0.0027 0.3225 -0.9466>}
 index_03_l {parent=index_02_l, loc=<2.5578 -0.123 -0.0252>, rot=<-0.002 0.0067 0.2558 -0.9667>}
 middle_metacarpal_l {parent=hand_l, loc=<2.1856 -0.1725 -0.5984>, rot=<0.006 0.0168 -0.0329 -0.9993>}
 middle_01_l {parent=middle_metacarpal_l, loc=<7.3275 -0.2447 0.0129>, rot=<-0.0208 0.0042 0.0858 -0.9961>}
 middle_02_l {parent=middle_01_l, loc=<4.0123 0.1401 0.1175>, rot=<-0.0218 0.0258 0.3512 -0.9357>}
 middle_03_l {parent=middle_02_l, loc=<2.6077 0.0534 0.1373>, rot=<0.0082 0.0188 0.1708 -0.9851>}
 pinky_metacarpal_l {parent=hand_l, loc=<1.7616 -0.9753 3.3814>, rot=<-0.0397 0.0814 -0.069 -0.9935>}
 pinky_01_l {parent=pinky_metacarpal_l, loc=<7.1709 -0.4831 -0.1094>, rot=<-0.0654 -0.0035 0.2242 -0.9723>}
 pinky_02_l {parent=pinky_01_l, loc=<2.8274 0.0893 -0.1043>, rot=<0.0037 0.0011 0.2854 -0.9584>}
 pinky_03_l {parent=pinky_02_l, loc=<2.2046 0.0307 -0.0078>, rot=<-0.0002 0.0098 0.1445 -0.9895>}
 ring_metacarpal_l {parent=hand_l, loc=<2.2074 -0.3625 1.3638>, rot=<-0.0284 0.0722 -0.0458 -0.9959>}
 ring_01_l {parent=ring_metacarpal_l, loc=<7.0295 -0.6197 -0.1446>, rot=<-0.0489 -0.0252 0.1892 -0.9804>}
 ring_02_l {parent=ring_01_l, loc=<3.6174 0.4647 -0.1027>, rot=<0 0 0.2806 -0.9598>}
 ring_03_l {parent=ring_02_l, loc=<2.4152 0.114 0.0199>, rot=<0 0 0.1189 -0.9929>}
 thumb_01_l {parent=hand_l, loc=<1.3118 -0.3191 -2.459>, rot=<0.6546 -0.4042 0.0664 -0.6354>}
 thumb_02_l {parent=thumb_01_l, loc=<5.9336 -0.0448 0.0401>, rot=<-0.0037 0.0213 0.298 -0.9543>}
 thumb_03_l {parent=thumb_02_l, loc=<2.2223 0.3682 -0.0907>, rot=<0.0118 -0.013 0.0066 -0.9998>}
 lowerarm_twist_01_l {parent=lowerarm_l, loc=<16.7573 0 0>, rot=<0 0 0 1>}
 lowerarm_twist_02_l {parent=lowerarm_l, loc=<11.5353 0 0>, rot=<0 0 0 1>}
 mstr_coat_arm_strap_l {parent=lowerarm_l, loc=<22.5928 -7.8867 -1.4276>, rot=<-0.0884 -0.1884 0.0701 -0.9756>}
 dyn_coat_arm_strap_01_l {parent=mstr_coat_arm_strap_l, loc=<0 0 0>, rot=<0 0 0 1>}
 dyn_coat_arm_strap_02_l {parent=dyn_coat_arm_strap_01_l, loc=<8.1 0 0>, rot=<-0.1618 -0.1472 0.1397 -0.9657>}
 dyn_coat_arm_strap_03_l {parent=dyn_coat_arm_strap_02_l, loc=<8.1002 0 0>, rot=<-0.2606 -0.0218 -0.0355 -0.9645>}
 upperarm_twist_01_l {parent=upperarm_l, loc=<7.2135 -0.1134 0>, rot=<0 0 0 1>}
 upperarm_twist_02_l {parent=upperarm_l, loc=<19.7998 -0.1134 0>, rot=<0 0 0 1>}
 armpit_01_l {parent=clavicle_l, loc=<16.0241 0.3337 -13.6384>, rot=<0.9512 0.0234 -0.3077 0.0093>}
 clavicle_r {parent=spine_05, loc=<4.1652 -0.3023 5.0942>, rot=<0.7289 0.0465 -0.6829 0.009>}
 upperarm_r {parent=clavicle_r, loc=<-11.2117 1.1486 0.0332>, rot=<0.0179 -0.2982 -0.028 -0.9539>}
 lowerarm_r {parent=upperarm_r, loc=<-26.5014 0.0835 -0.2605>, rot=<0.0036 0.0161 -0.2298 -0.9731>}
 hand_r {parent=lowerarm_r, loc=<-23.1357 1.1606 0.2612>, rot=<-0.6704 -0.0028 0.0567 -0.7398>}
 index_metacarpal_r {parent=hand_r, loc=<-1.6534 -0.0181 2.9601>, rot=<0.0158 -0.0124 -0.0173 -0.9996>}
 index_01_r {parent=index_metacarpal_r, loc=<-8.3008 -0.0357 0.0524>, rot=<-0.0044 0.0485 0.0924 -0.9945>}
 index_02_r {parent=index_01_r, loc=<-3.3589 0.3003 -0.0453>, rot=<-0.002 0.0027 0.3225 -0.9466>}
 index_03_r {parent=index_02_r, loc=<-2.5578 0.123 0.0252>, rot=<-0.002 0.0067 0.2558 -0.9667>}
 middle_metacarpal_r {parent=hand_r, loc=<-2.1856 0.1726 0.5984>, rot=<0.006 0.0168 -0.0329 -0.9993>}
 middle_01_r {parent=middle_metacarpal_r, loc=<-7.3273 0.2447 -0.0129>, rot=<-0.0208 0.0042 0.0858 -0.9961>}
 middle_02_r {parent=middle_01_r, loc=<-4.0124 -0.14 -0.1175>, rot=<-0.0218 0.0258 0.3512 -0.9357>}
 middle_03_r {parent=middle_02_r, loc=<-2.6077 -0.0533 -0.1373>, rot=<0.0082 0.0188 0.1708 -0.9851>}
 pinky_metacarpal_r {parent=hand_r, loc=<-1.7617 0.9753 -3.3814>, rot=<-0.0397 0.0814 -0.069 -0.9935>}
 pinky_01_r {parent=pinky_metacarpal_r, loc=<-7.1707 0.4832 0.1094>, rot=<-0.0654 -0.0035 0.2242 -0.9723>}
 pinky_02_r {parent=pinky_01_r, loc=<-2.8274 -0.0892 0.1043>, rot=<0.0037 0.0011 0.2854 -0.9584>}
 pinky_03_r {parent=pinky_02_r, loc=<-2.2045 -0.031 0.0078>, rot=<-0.0002 0.0098 0.1445 -0.9895>}
 ring_metacarpal_r {parent=hand_r, loc=<-2.2073 0.3627 -1.3638>, rot=<-0.0284 0.0722 -0.0458 -0.9959>}
 ring_01_r {parent=ring_metacarpal_r, loc=<-7.0297 0.6197 0.1446>, rot=<-0.0489 -0.0252 0.1892 -0.9804>}
 ring_02_r {parent=ring_01_r, loc=<-3.6171 -0.4648 0.1027>, rot=<0 0 0.2806 -0.9598>}
 ring_03_r {parent=ring_02_r, loc=<-2.4153 -0.1141 -0.0199>, rot=<0 0 0.1189 -0.9929>}
 thumb_01_r {parent=hand_r, loc=<-1.3119 0.3192 2.459>, rot=<0.6546 -0.4042 0.0664 -0.6354>}
 thumb_02_r {parent=thumb_01_r, loc=<-5.9335 0.0447 -0.0401>, rot=<-0.0037 0.0213 0.298 -0.9543>}
 thumb_03_r {parent=thumb_02_r, loc=<-2.2223 -0.3682 0.0907>, rot=<0.0118 -0.013 0.0066 -0.9998>}
 lowerarm_twist_01_r {parent=lowerarm_r, loc=<-16.7573 0 0>, rot=<0 0 0 1>}
 lowerarm_twist_02_r {parent=lowerarm_r, loc=<-11.5354 0 0>, rot=<0 0 0 1>}
 mstr_coat_arm_strap_r {parent=lowerarm_r, loc=<-22.8848 7.9197 1.4495>, rot=<0.2185 -0.0313 0.9729 0.0689>}
 dyn_coat_arm_strap_01_r {parent=mstr_coat_arm_strap_r, loc=<0 0 0>, rot=<0 0 0 1>}
 dyn_coat_arm_strap_02_r {parent=dyn_coat_arm_strap_01_r, loc=<8.7596 0 0>, rot=<0.2484 0.1287 0.1149 -0.9532>}
 dyn_coat_arm_strap_03_r {parent=dyn_coat_arm_strap_02_r, loc=<8.75 0 0>, rot=<0.3059 0.0231 -0.03 -0.9513>}
 upperarm_twist_01_r {parent=upperarm_r, loc=<-7.2135 0.1134 0>, rot=<0 0 0 1>}
 upperarm_twist_02_r {parent=upperarm_r, loc=<-19.7998 0.1134 0>, rot=<0 0 0 1>}
 armpit_01_r {parent=clavicle_r, loc=<-16.0241 -0.3337 13.6383>, rot=<0.9512 0.0234 -0.3077 0.0093>}
 neck_01 {parent=spine_05, loc=<11.3775 -0.6396 0>, rot=<0 0 -0.2003 -0.9797>}
 neck_02 {parent=neck_01, loc=<4.5627 0 0>, rot=<0 0 0.0504 -0.9987>}
 head {parent=neck_02, loc=<5.3914 0 0>, rot=<0.0393 -0.0042 0.1062 -0.9936>}
 dyn_mstr_choker_ring {parent=neck_01, loc=<3.1049 5.3358 -0.0006>, rot=<-0.0004 0.0006 -0.6129 -0.7902>}
 dyn_choker_ring {parent=dyn_mstr_choker_ring, loc=<0.4409 0 0>, rot=<-0.0005 -0.0005 -0.6593 -0.7519>}
 necklace_01 {parent=spine_05, loc=<6.815 5.7969 0.0004>, rot=<0 0 0.9683 0.2498>}
 mstr_back_d_ring {parent=spine_05, loc=<6.2457 -10.1004 0>, rot=<0.1748 -0.9846 0 0>}
 back_d_ring {parent=mstr_back_d_ring, loc=<0 0 0>, rot=<0 0 0 1>}
 mstr_under_arm_l {parent=spine_04, loc=<0 0 -16.3907>, rot=<0.0416 -0.9991 0 0>}
 under_arm_01_l {parent=mstr_under_arm_l, loc=<0 0 0>, rot=<-0.7071 0 0 -0.7071>}
 under_arm_02_l {parent=under_arm_01_l, loc=<9.986 0 0>, rot=<0 0 0 1>}
 dyn_mstr_under_arm_l {parent=under_arm_02_l, loc=<9.985 0 0>, rot=<0 0 0 1>}
 dyn_under_arm_01_l {parent=dyn_mstr_under_arm_l, loc=<0 0 0>, rot=<0 0 0 1>}
 dyn_under_arm_02_l {parent=dyn_under_arm_01_l, loc=<7.806 0 0>, rot=<0 0 0 1>}
 dyn_under_arm_03_l {parent=dyn_under_arm_02_l, loc=<7.806 0 0>, rot=<0 0 0 1>}
 mstr_mid_back {parent=spine_04, loc=<0.899 -10.7791 0>, rot=<0.0416 -0.9991 0 0>}
 mid_back_01 {parent=mstr_mid_back, loc=<0 0 0>, rot=<0 0 0 1>}
 mid_back_02 {parent=mid_back_01, loc=<9.986 0 0>, rot=<0 0 0 1>}
 dyn_mstr_mid_back {parent=mid_back_02, loc=<9.9861 0 0>, rot=<0 0 0 1>}
 dyn_mid_back_01 {parent=dyn_mstr_mid_back, loc=<0 0 0>, rot=<0 0 0 1>}
 dyn_mid_back_02 {parent=dyn_mid_back_01, loc=<7.8061 0 0>, rot=<0 0 0 1>}
 dyn_mid_back_03 {parent=dyn_mid_back_02, loc=<7.8061 0 0>, rot=<0 0 0 1>}
 mstr_frnt_torso_l {parent=spine_04, loc=<-0.9879 11.8441 -7.8712>, rot=<0.0416 -0.9991 0 0>}
 frnt_torso_01_l {parent=mstr_frnt_torso_l, loc=<0 0 0>, rot=<-0.1736 0 0 -0.9848>}
 frnt_torso_02_l {parent=frnt_torso_01_l, loc=<9.986 0 0>, rot=<0 0 0 1>}
 dyn_mstr_frnt_torso_l {parent=frnt_torso_02_l, loc=<9.985 0 0>, rot=<0 0 0 1>}
 dyn_frnt_torso_01_l {parent=dyn_mstr_frnt_torso_l, loc=<0 0 0>, rot=<0 0 0 1>}
 dyn_frnt_torso_02_l {parent=dyn_frnt_torso_01_l, loc=<7.806 0 0>, rot=<0 0 0 1>}
 dyn_frnt_torso_03_l {parent=dyn_frnt_torso_02_l, loc=<7.806 0 0>, rot=<0 0 0 1>}
 mstr_back_l {parent=spine_04, loc=<0.7709 -9.2436 -10.216>, rot=<0.0416 -0.9991 0 0>}
 back_01_l {parent=mstr_back_l, loc=<0 0 0>, rot=<0.9763 0 0 0.2164>}
 back_02_l {parent=back_01_l, loc=<9.986 0 0>, rot=<0 0 0 1>}
 dyn_mstr_back_l {parent=back_02_l, loc=<9.985 0 0>, rot=<0 0 0 1>}
 dyn_back_01_l {parent=dyn_mstr_back_l, loc=<0 0 0>, rot=<0 0 0 1>}
 dyn_back_02_l {parent=dyn_back_01_l, loc=<7.806 0 0>, rot=<0 0 0 1>}
 dyn_back_03_l {parent=dyn_back_02_l, loc=<7.806 0 0>, rot=<0 0 0 1>}
 mstr_back_r {parent=spine_04, loc=<0.7709 -9.2436 10.216>, rot=<0.9991 0.0416 0 0>}
 back_01_r {parent=mstr_back_r, loc=<0 0 0>, rot=<0.9763 0 0 0.2164>}
 back_02_r {parent=back_01_r, loc=<-9.986 0 0>, rot=<0 0 0 -1>}
 dyn_mstr_back_r {parent=back_02_r, loc=<-9.985 0 0>, rot=<0 0 0 -1>}
 dyn_back_01_r {parent=dyn_mstr_back_r, loc=<0 0 0>, rot=<0 0 0 -1>}
 dyn_back_02_r {parent=dyn_back_01_r, loc=<-7.806 0 0>, rot=<0 0 0 -1>}
 dyn_back_03_r {parent=dyn_back_02_r, loc=<-7.806 0 0>, rot=<0 0 0 -1>}
 mstr_under_arm_r {parent=spine_04, loc=<0 0 16.3907>, rot=<0.9991 0.0416 0 0>}
 under_arm_01_r {parent=mstr_under_arm_r, loc=<0 0 0>, rot=<-0.7071 0 0 -0.7071>}
 under_arm_02_r {parent=under_arm_01_r, loc=<-9.986 0 0>, rot=<0 0 0 1>}
 dyn_mstr_under_arm_r {parent=under_arm_02_r, loc=<-9.985 0 0>, rot=<0 0 0 1>}
 dyn_under_arm_01_r {parent=dyn_mstr_under_arm_r, loc=<0 0 0>, rot=<0 0 0 1>}
 dyn_under_arm_02_r {parent=dyn_under_arm_01_r, loc=<-7.806 0 0>, rot=<0 0 0 1>}
 dyn_under_arm_03_r {parent=dyn_under_arm_02_r, loc=<-7.806 0 0>, rot=<0 0 0 1>}
 mstr_frnt_torso_r {parent=spine_04, loc=<-0.9879 11.8441 7.8712>, rot=<0.9991 0.0416 0 0>}
 frnt_torso_01_r {parent=mstr_frnt_torso_r, loc=<0 0 0>, rot=<-0.1736 0 0 -0.9848>}
 frnt_torso_02_r {parent=frnt_torso_01_r, loc=<-9.986 0 0>, rot=<0 0 0 1>}
 dyn_mstr_frnt_torso_r {parent=frnt_torso_02_r, loc=<-9.985 0 0>, rot=<0 0 0 1>}
 dyn_frnt_torso_01_r {parent=dyn_mstr_frnt_torso_r, loc=<0 0 0>, rot=<0 0 0 1>}
 dyn_frnt_torso_02_r {parent=dyn_frnt_torso_01_r, loc=<-7.806 0 0>, rot=<0 0 0 1>}
 dyn_frnt_torso_03_r {parent=dyn_frnt_torso_02_r, loc=<-7.806 0 0>, rot=<0 0 0 1>}
 mstr_zipper_l {parent=spine_03, loc=<4.9633 8.484 -13.1599>, rot=<0.1695 0.2899 0.9417 0.0215>}
 dyn_zipper_l {parent=mstr_zipper_l, loc=<0 0 0>, rot=<0 0 0 1>}
 mstr_zipper_r {parent=spine_03, loc=<4.9632 8.484 13.1598>, rot=<-0.2899 0.1695 0.0215 -0.9417>}
 dyn_zipper_r {parent=mstr_zipper_r, loc=<0 0 0>, rot=<0 0 0 1>}
 thigh_l {parent=pelvis, loc=<-2.5624 -0.0642 -10.3834>, rot=<0.0004 0.0055 0.0155 -0.9999>}
 calf_l {parent=thigh_l, loc=<-39.2091 -0.1238 -0.6795>, rot=<0.0006 -0.0161 -0.0391 -0.9991>}
 calf_twist_01_l {parent=calf_l, loc=<-34.2864 -1.9197 -0.5794>, rot=<0.0237 0.0095 -0.0035 -0.9997>}
 calf_twist_02_l {parent=calf_l, loc=<-17.5927 -1.7943 -0.2638>, rot=<0.0237 0.0095 -0.0035 -0.9997>}
 foot_l {parent=calf_l, loc=<-43.624 -2.818 -0.713>, rot=<0.044 0.0061 0.0367 -0.9983>}
 ball_l {parent=foot_l, loc=<-4.6464 10.3094 -0.3137>, rot=<0.0001 -0.0001 -0.7063 -0.7079>}
 thigh_twist_01_l {parent=thigh_l, loc=<-23.0764 0 0>, rot=<0.0013 0 0 -1>}
 thigh_r {parent=pelvis, loc=<-2.5623 -0.0642 10.3834>, rot=<0.0055 -0.0004 0.9999 0.0155>}
 calf_r {parent=thigh_r, loc=<39.2093 0.1238 0.6795>, rot=<0.0006 -0.0161 -0.0391 -0.9991>}
 calf_twist_01_r {parent=calf_r, loc=<34.2866 1.9197 0.5794>, rot=<0.0237 0.0095 -0.0035 -0.9997>}
 calf_twist_02_r {parent=calf_r, loc=<17.5927 1.7943 0.2638>, rot=<0.0237 0.0095 -0.0035 -0.9997>}
 foot_r {parent=calf_r, loc=<43.6242 2.818 0.713>, rot=<0.044 0.0061 0.0367 -0.9983>}
 ball_r {parent=foot_r, loc=<4.6464 -10.3094 0.3137>, rot=<0.0001 -0.0001 -0.7063 -0.7079>}
 thigh_twist_01_r {parent=thigh_r, loc=<23.0764 0 0>, rot=<0.0013 0 0 -1>}
 mstr_coat_strap_r {parent=pelvis, loc=<-5.9548 4.3317 15.8998>, rot=<-0.282 -0.6636 0.6641 0.1977>}
 dyn_coat_strap_01_r {parent=mstr_coat_strap_r, loc=<0 0 0>, rot=<0 0 0 1>}
 dyn_coat_strap_02_r {parent=dyn_coat_strap_01_r, loc=<6.1429 0 0>, rot=<-0.0272 0.2166 0.0869 -0.972>}
 dyn_coat_strap_03_r {parent=dyn_coat_strap_02_r, loc=<5.2232 0 0>, rot=<0.1053 0.0984 -0.0107 -0.9895>}
 mstr_coat_strap_l {parent=pelvis, loc=<-5.8278 4.0682 -15.9731>, rot=<-0.2019 -0.707 -0.6576 -0.1642>}
 dyn_coat_strap_01_l {parent=mstr_coat_strap_l, loc=<0 0 0>, rot=<0 0 0 1>}
 dyn_coat_strap_02_l {parent=dyn_coat_strap_01_l, loc=<8.9013 0 0>, rot=<-0.2708 -0.1621 0.085 -0.9451>}
 dyn_coat_strap_03_l {parent=dyn_coat_strap_02_l, loc=<9.3449 0 0>, rot=<-0.3304 -0.0247 0.0349 -0.9429>}
 clavicle_l_ctrl {controlType=EULER_TRANSFORM, type=Control, loc=<0 0 0>, rot=<0 0 0 1>}
 upperarm_l_ctrl {controlType=EULER_TRANSFORM, parent=clavicle_l_ctrl, type=Control, loc=<80.0272 6.5919 -108.1636>, rot=<-0.0165 -0.0339 0.0185 0.9991>}
 clavicle_r_ctrl {controlType=EULER_TRANSFORM, type=Control, loc=<0 0 0>, rot=<0 0 0 1>}
 upperarm_r_ctrl {controlType=EULER_TRANSFORM, parent=clavicle_r_ctrl, type=Control, loc=<-79.9936 -8.8878 108.0674>, rot=<-0.8018 0.0395 -0.5963 -0.0008>}
]
variables [is_crouching {type=bool, default=False}]
