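# Animation Blueprint for the female large purple leather jacket body cosmetic, adding procedural and secondary motion to jacket bones.
# It copies the attached parent's pose, then applies spine constraints and a torso Control Rig, cached for further weighted constraints and look-ats.
# Downstream of the RBAN cache, constraints place zippers and straps, and AnimDynamics swings them, plus rings.
# Sprint, skydive, melee, crouch and spine-twist variables blend corrective bone adjustments.
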AnimBlueprint /BRCosmetics/Characters/Player/Female/Large/Bodies/F_LRG_Leather_Jacket_Purple/Meshes/F_LRG_Leather_Jacket_Purple_AnimBP
target_skeleton=&/F_LRG_Leather_Jacket_Purple_Skeleton
graphs [
 AnimGraph {nodes [
   n1 = "Constraint - back_02_l (PP)"!(ComponentPose, bAlphaBoolEnabled:bool=True, ConstraintWeights_0=0.5, ConstraintWeights_1:float=0.5) -> (Pose) {
    alphaInputType=Bool
    bAlphaBoolEnabled=false
    alphaBoolBlend {blendOption=Linear}
    boneToModify {boneName=back_02_l}
    constraintSetup [{targetBone {boneName=spine_03}}, {targetBone {boneName=back_01_l}}]
    constraintWeights [1,1]
   }
   n2 = "Constraint - back_02_r (PP)"!(ComponentPose=@n1, bAlphaBoolEnabled:bool=True, ConstraintWeights_0=0.5, ConstraintWeights_1:float=0.5) -> (Pose) {
    alphaInputType=Bool
    bAlphaBoolEnabled=false
    alphaBoolBlend {blendOption=Linear}
    boneToModify {boneName=back_02_r}
    constraintSetup [{targetBone {boneName=spine_03}}, {targetBone {boneName=back_01_r}}]
    constraintWeights [1,1]
   }
   "Constraint - mid_back_02 (PP)"!(ComponentPose=@n2, bAlphaBoolEnabled:bool=True, ConstraintWeights_0=0.5, ConstraintWeights_1:float=0.5) -> (Pose) {
    alphaInputType=Bool
    bAlphaBoolEnabled=false
    alphaBoolBlend {blendOption=Linear}
    boneToModify {boneName=mid_back_02}
    constraintSetup [{targetBone {boneName=spine_03}}, {targetBone {boneName=mid_back_01}}]
    constraintWeights [1,1]
   }
   n3 = "Constraint - dyn_mstr_back_l (PP)"!(ComponentPose, bAlphaBoolEnabled:bool=True, ConstraintWeights_0=1, ConstraintWeights_1:float=0) -> (Pose) {
    alphaInputType=Bool
    bAlphaBoolEnabled=false
    alphaBoolBlend {blendOption=Linear}
    boneToModify {boneName=dyn_mstr_back_l}
    constraintSetup [{targetBone {boneName=spine_01}}, {targetBone {boneName=back_02_l}}]
    constraintWeights [1,1]
   }
   n4 = "Constraint - dyn_mstr_back_r (PP)"!(ComponentPose=@n3, bAlphaBoolEnabled:bool=True, ConstraintWeights_0=1, ConstraintWeights_1:float=0) -> (Pose) {
    alphaInputType=Bool
    bAlphaBoolEnabled=false
    alphaBoolBlend {blendOption=Linear}
    boneToModify {boneName=dyn_mstr_back_r}
    constraintSetup [{targetBone {boneName=spine_01}}, {targetBone {boneName=back_02_r}}]
    constraintWeights [1,1]
   }
   "Constraint - dyn_mstr_mid_back (PP)"!(ComponentPose=@n4, bAlphaBoolEnabled:bool=True, ConstraintWeights_0=1, ConstraintWeights_1:float=0) -> (Pose) {
    alphaInputType=Bool
    bAlphaBoolEnabled=false
    alphaBoolBlend {blendOption=Linear}
    boneToModify {boneName=dyn_mstr_mid_back}
    constraintSetup [{targetBone {boneName=spine_01}}, {targetBone {boneName=mid_back_02}}]
    constraintWeights [1,1]
   }
   "Constraint - dyn_mstr_back_l (R)"(Alpha=0.2, ConstraintWeights_0=1) {
    alphaBoolBlend {blendOption=Linear}
    boneToModify {boneName=dyn_mstr_back_l}
    constraintSetup [{transformType=Rotation, targetBone {boneName=spine_02}}]
    constraintWeights [1]
   }
   "Constraint - dyn_mstr_mid_back (R)"(Alpha=0.2, ConstraintWeights_0=1) {
    alphaBoolBlend {blendOption=Linear}
    boneToModify {boneName=dyn_mstr_mid_back}
    constraintSetup [{transformType=Rotation, targetBone {boneName=spine_02}}]
    constraintWeights [1]
   }
   "Constraint - dyn_mstr_back_r (R)"(Alpha=0.2, ConstraintWeights_0=1) {
    alphaBoolBlend {blendOption=Linear}
    boneToModify {boneName=dyn_mstr_back_r}
    constraintSetup [{transformType=Rotation, targetBone {boneName=spine_02}}]
    constraintWeights [1]
   }
   "Transform (Modify) Bone - Bone: dyn_coat_strap_02_r"(Alpha=1, Translation="-5,0,0", Rotation="0,0,0", Scale="1,1,1") {
    translationMode=BMM_Additive
    translationSpace=BCS_BoneSpace
    alphaBoolBlend {blendOption=Linear}
    boneToModify {boneName=dyn_coat_strap_02_r}
    translation {x=-5}
   }
   n5 = "Constraint - mstr_zipper_l (PPPPPPPP)"!(
    ComponentPose=Local To Component(LocalPose=Use cached pose 'RBANCache'())
    Alpha=1
    ConstraintWeights_0=0.0135
    ConstraintWeights_1:float=0.0059
    ConstraintWeights_2:float=0.0936
    ConstraintWeights_3:float=0.0595
    ConstraintWeights_4:float=0.7372
    ConstraintWeights_5:float=0.0413
    ConstraintWeights_6:float=0.0448
    ConstraintWeights_7:float=0.0042
   ) -> (Pose) {
    alphaBoolBlend {blendOption=Linear}
    boneToModify {boneName=mstr_zipper_l}
    constraintSetup [
     {targetBone {boneName=upperarm_twist_01_l}}
     {targetBone {boneName=spine_04}}
     {targetBone {boneName=spine_05}}
     {targetBone {boneName=frnt_torso_02_l}}
     {targetBone {boneName=frnt_torso_01_l}}
     {targetBone {boneName=under_arm_02_l}}
     {targetBone {boneName=under_arm_01_l}}
     {targetBone {boneName=armpit_01_l}}
    ]
    constraintWeights [1,1,1,1,1,1,1,1]
   }
   n6 = "Constraint - mstr_zipper_r (PPPPPPPP)"!(
    ComponentPose=@n5
    Alpha=1
    ConstraintWeights_0=0.0081
    ConstraintWeights_1:float=0.0045
    ConstraintWeights_2:float=0.116
    ConstraintWeights_3:float=0.0186
    ConstraintWeights_4:float=0.0499
    ConstraintWeights_5:float=0.0762
    ConstraintWeights_6:float=0.0744
    ConstraintWeights_7:float=0.6523
   ) -> (Pose) {
    alphaBoolBlend {blendOption=Linear}
    boneToModify {boneName=mstr_zipper_r}
    constraintSetup [
     {targetBone {boneName=spine_04}}
     {targetBone {boneName=spine_03}}
     {targetBone {boneName=spine_05}}
     {targetBone {boneName=upperarm_twist_01_r}}
     {targetBone {boneName=under_arm_02_r}}
     {targetBone {boneName=under_arm_01_r}}
     {targetBone {boneName=frnt_torso_02_r}}
     {targetBone {boneName=frnt_torso_01_r}}
    ]
    constraintWeights [1,1,1,1,1,1,1,1]
   }
   n7 = "Constraint - mstr_coat_strap_r (PPPPPPPPP)"!(
    ComponentPose=@n6
    Alpha=1
    ConstraintWeights_0=0.3976
    ConstraintWeights_1:float=0.3193
    ConstraintWeights_2:float=0.1277
    ConstraintWeights_3:float=0.051
    ConstraintWeights_4:float=0.0472
    ConstraintWeights_5:float=0.0241
    ConstraintWeights_6:float=0.0213
    ConstraintWeights_7:float=0.0099
    ConstraintWeights_8:float=0.002
   ) -> (Pose) {
    alphaBoolBlend {blendOption=Linear}
    boneToModify {boneName=mstr_coat_strap_r}
    constraintSetup [
     {targetBone {boneName=dyn_under_arm_03_r}}
     {targetBone {boneName=dyn_coat_strap_01_r}}
     {targetBone {boneName=dyn_frnt_torso_02_r}}
     {targetBone {boneName=dyn_under_arm_01_r}}
     {targetBone {boneName=dyn_under_arm_02_r}}
     {targetBone {boneName=dyn_frnt_torso_01_r}}
     {targetBone {boneName=dyn_frnt_torso_03_r}}
     {targetBone {boneName=dyn_back_03_r}}
     {targetBone {boneName=frnt_torso_02_r}}
    ]
    constraintWeights [1,1,1,1,1,1,1,1,1]
   }
   n8 = "Constraint - mstr_coat_strap_l (PPPPPPPPP)"!(
    ComponentPose=@n7
    Alpha=1
    ConstraintWeights_0=0.3505
    ConstraintWeights_1:float=0.1431
    ConstraintWeights_2:float=0.1421
    ConstraintWeights_3:float=0.1385
    ConstraintWeights_4:float=0.1063
    ConstraintWeights_5:float=0.0661
    ConstraintWeights_6:float=0.0429
    ConstraintWeights_7:float=0.0056
    ConstraintWeights_8:float=0.0049
   ) -> (Pose) {
    alphaBoolBlend {blendOption=Linear}
    boneToModify {boneName=mstr_coat_strap_l}
    constraintSetup [
     {targetBone {boneName=dyn_coat_strap_01_l}}
     {targetBone {boneName=dyn_frnt_torso_02_l}}
     {targetBone {boneName=dyn_under_arm_01_l}}
     {targetBone {boneName=dyn_under_arm_03_l}}
     {targetBone {boneName=dyn_under_arm_02_l}}
     {targetBone {boneName=dyn_frnt_torso_01_l}}
     {targetBone {boneName=dyn_frnt_torso_03_l}}
     {targetBone {boneName=frnt_torso_02_l}}
     {targetBone {boneName=under_arm_02_l}}
    ]
    constraintWeights [1,1,1,1,1,1,1,1,1]
   }
   n9 = "Constraint - dyn_coat_strap_02_r (PPPPPPP)"!(
    ComponentPose=@n8
    Alpha=1
    ConstraintWeights_0=0.3923
    ConstraintWeights_1:float=0.0082
    ConstraintWeights_2:float=0.0077
    ConstraintWeights_3:float=0.4225
    ConstraintWeights_4:float=0.1214
    ConstraintWeights_5:float=0.0172
    ConstraintWeights_6:float=0.0307
   ) -> (Pose) {
    alphaBoolBlend {blendOption=Linear}
    boneToModify {boneName=dyn_coat_strap_02_r}
    constraintSetup [
     {targetBone {boneName=dyn_under_arm_03_r}}
     {targetBone {boneName=dyn_under_arm_02_r}}
     {targetBone {boneName=dyn_under_arm_01_r}}
     {targetBone {boneName=dyn_frnt_torso_03_r}}
     {targetBone {boneName=dyn_frnt_torso_02_r}}
     {targetBone {boneName=dyn_frnt_torso_01_r}}
     {targetBone {boneName=dyn_back_03_r}}
    ]
    constraintWeights [1,1,1,1,1,1,1]
   }
   n10 = "Constraint - dyn_coat_strap_02_l (PPPPPPPP)"!(
    ComponentPose=@n9
    Alpha=1
    ConstraintWeights_0=0.0384
    ConstraintWeights_1:float=0.466
    ConstraintWeights_2:float=0.1039
    ConstraintWeights_3:float=0.03
    ConstraintWeights_4:float=0.0011
    ConstraintWeights_5:float=0.3157
    ConstraintWeights_6:float=0.0173
    ConstraintWeights_7:float=0.0275
   ) -> (Pose) {
    alphaBoolBlend {blendOption=Linear}
    boneToModify {boneName=dyn_coat_strap_02_l}
    constraintSetup [
     {targetBone {boneName=dyn_back_03_l}}
     {targetBone {boneName=dyn_frnt_torso_03_l}}
     {targetBone {boneName=dyn_frnt_torso_02_l}}
     {targetBone {boneName=dyn_frnt_torso_01_l}}
     {targetBone {boneName=dyn_mid_back_03}}
     {targetBone {boneName=dyn_under_arm_03_l}}
     {targetBone {boneName=dyn_under_arm_02_l}}
     {targetBone {boneName=dyn_under_arm_01_l}}
    ]
    constraintWeights [1,1,1,1,1,1,1,1]
   }
   n11 = AND Boolean(A=Get bIsSprinting(), B=NOT Boolean(A=Get isJacketMovedtoSide() -> (isJacketMovedtoSide:bool)))
   n12 = "Constraint - dyn_frnt_torso_01_l (R)"!(ComponentPose=@n10, bAlphaBoolEnabled:bool=@n11, ConstraintWeights_0=1) -> (Pose) {
    alphaInputType=Bool
    alphaBoolBlend {blendInTime=0.2, blendOption=Linear, blendOutTime=0.2}
    boneToModify {boneName=dyn_frnt_torso_01_l}
    constraintSetup [{transformType=Rotation, targetBone {boneName=thigh_l}}]
    constraintWeights [1]
   }
   n13 = "Constraint - dyn_frnt_torso_02_l (R)"!(ComponentPose=@n12, bAlphaBoolEnabled:bool=@n11, ConstraintWeights_0=1) -> (Pose) {
    alphaInputType=Bool
    alphaBoolBlend {blendInTime=0.2, blendOption=Linear, blendOutTime=0.2}
    boneToModify {boneName=dyn_frnt_torso_02_l}
    constraintSetup [{transformType=Rotation, targetBone {boneName=thigh_l}}]
    constraintWeights [1]
   }
   n14 = "Constraint - dyn_frnt_torso_03_l (R)"!(ComponentPose=@n13, bAlphaBoolEnabled:bool=@n11, ConstraintWeights_0=1) -> (Pose) {
    alphaInputType=Bool
    alphaBoolBlend {blendInTime=0.2, blendOption=Linear, blendOutTime=0.2}
    boneToModify {boneName=dyn_frnt_torso_03_l}
    constraintSetup [{transformType=Rotation, targetBone {boneName=thigh_l}}]
    constraintWeights [1]
   }
   n15 = "Constraint - dyn_frnt_torso_01_r (R)"!(ComponentPose=@n14, bAlphaBoolEnabled:bool=@n11, ConstraintWeights_0=1) -> (Pose) {
    alphaInputType=Bool
    alphaBoolBlend {blendInTime=0.2, blendOption=Linear, blendOutTime=0.2}
    boneToModify {boneName=dyn_frnt_torso_01_r}
    constraintSetup [{transformType=Rotation, targetBone {boneName=thigh_r}}]
    constraintWeights [1]
   }
   n16 = "Constraint - dyn_frnt_torso_02_r (R)"!(ComponentPose=@n15, bAlphaBoolEnabled:bool=@n11, ConstraintWeights_0=1) -> (Pose) {
    alphaInputType=Bool
    alphaBoolBlend {blendInTime=0.2, blendOption=Linear, blendOutTime=0.2}
    boneToModify {boneName=dyn_frnt_torso_02_r}
    constraintSetup [{transformType=Rotation, targetBone {boneName=thigh_r}}]
    constraintWeights [1]
   }
   n17 = "Constraint - dyn_frnt_torso_03_r (R)"!(ComponentPose=@n16, bAlphaBoolEnabled:bool=@n11, ConstraintWeights_0=1) -> (Pose) {
    alphaInputType=Bool
    alphaBoolBlend {blendInTime=0.2, blendOption=Linear, blendOutTime=0.2}
    boneToModify {boneName=dyn_frnt_torso_03_r}
    constraintSetup [{transformType=Rotation, targetBone {boneName=thigh_r}}]
    constraintWeights [1]
   }
   n18 = "Transform (Modify) Bone - Bone: dyn_coat_strap_02_r"(ComponentPose=@n17, Alpha=1, Translation="-5,0,0", Rotation="0,0,0", Scale="1,1,1") {
    translationMode=BMM_Additive
    translationSpace=BCS_BoneSpace
    alphaBoolBlend {blendOption=Linear}
    boneToModify {boneName=dyn_coat_strap_02_r}
    translation {x=-5}
   }
   n19 = Anim Dynamics - Bone: back_d_ring(ComponentPose=@n18, Alpha=1, ExternalForce="0,0,0") {
    angularDampingOverride=1
    angularSpringConstant=20000
    bAngularSpring=true
    bOverrideAngularDamping=true
    bUseGravityOverride=true
    gravityScale=0.35
    simulationSpace=RootRelative
    alphaBoolBlend {blendOption=Linear}
    boundBone {boneName=back_d_ring}
    componentLinearAccScale=<0.5 0.5 0.5>
    componentLinearVelScale=<1 1 1>
    gravityOverride {z=-980}
    physicsBodyDefinitions [{
      sphereCollisionRadius=0
      boundBone {boneName=back_d_ring}
      boxExtents=<20 30 30>
      constraintSetup {angularLimitsMax=<1 1 0>, angularLimitsMin=<-1 -1 -30>, angularTarget=<1 0 0>}
      localJointOffset=<4 0 0>
     }]
   }
   n20 = Anim Dynamics - Bone: dyn_choker_ring(ComponentPose=@n19, Alpha=1, ExternalForce="0,0,0") {
    angularDampingOverride=1
    angularSpringConstant=1000
    bAngularSpring=true
    bOverrideAngularDamping=true
    bUseGravityOverride=true
    gravityScale=0.35
    simulationSpace=RootRelative
    alphaBoolBlend {blendOption=Linear}
    boundBone {boneName=dyn_choker_ring}
    componentAppliedLinearAccClamp=<3000 3000 3000>
    componentLinearAccScale=<0.2 0.2 0.3>
    componentLinearVelScale=<0.2 0.2 0.3>
    gravityOverride {z=-980}
    physicsBodyDefinitions [{
      sphereCollisionRadius=0
      boundBone {boneName=dyn_choker_ring}
      boxExtents=<10 15 15>
      constraintSetup {angularLimitsMax=<1 1 2>, angularLimitsMin=<-1 -1 -30>, angularTarget=<1 0 0>}
      localJointOffset=<4 0 0>
     }]
   }
   n21 = Anim Dynamics - Bone: dyn_zipper_l(ComponentPose=@n20, Alpha=1, ExternalForce="0,0,0") {
    angularDampingOverride=0.999
    angularSpringConstant=500
    bAngularSpring=true
    bOverrideAngularDamping=true
    bUseGravityOverride=true
    gravityScale=0.35
    simulationSpace=RootRelative
    alphaBoolBlend {blendOption=Linear}
    boundBone {boneName=dyn_zipper_l}
    componentAppliedLinearAccClamp=<3000 3000 3000>
    componentLinearAccScale=<0.3 0.3 0.4>
    componentLinearVelScale=<0.3 0.3 0.4>
    gravityOverride {z=-980}
    physicsBodyDefinitions [{
      sphereCollisionRadius=0
      boundBone {boneName=dyn_zipper_l}
      boxExtents=<8 15 15>
      constraintSetup {angularLimitsMax=<2 20 0>, angularLimitsMin=<-2 -20 -35>, angularTarget=<1 0 0>}
      localJointOffset=<4 0 0>
     }]
   }
   n22 = Anim Dynamics - Bone: dyn_zipper_r(ComponentPose=@n21, Alpha=1, ExternalForce="0,0,0") {
    angularDampingOverride=0.999
    angularSpringConstant=500
    bAngularSpring=true
    bOverrideAngularDamping=true
    bUseGravityOverride=true
    gravityScale=0.35
    simulationSpace=RootRelative
    alphaBoolBlend {blendOption=Linear}
    boundBone {boneName=dyn_zipper_r}
    componentAppliedLinearAccClamp=<3000 3000 3000>
    componentLinearAccScale=<0.3 0.3 0.4>
    componentLinearVelScale=<0.3 0.3 0.4>
    gravityOverride {z=-980}
    physicsBodyDefinitions [{
      sphereCollisionRadius=0
      boundBone {boneName=dyn_zipper_r}
      boxExtents=<8 15 15>
      constraintSetup {angularLimitsMax=<2 20 0>, angularLimitsMin=<-2 -20 -35>, angularTarget=<1 0 0>}
      localJointOffset=<-4 0 0>
     }]
   }
   n23 = "Transform (Modify) Bone - Bone: dyn_coat_strap_02_l"!(ComponentPose=@n22, bAlphaBoolEnabled:bool=True, Translation="0,0,0", Rotation="20,20,0", Scale="1,1,1") -> (Pose) {
    alphaInputType=Bool
    bAlphaBoolEnabled=false
    rotationMode=BMM_Additive
    rotationSpace=BCS_BoneSpace
    translationMode=BMM_Additive
    translationSpace=BCS_BoneSpace
    alphaBoolBlend {blendOption=Linear}
    boneToModify {boneName=dyn_coat_strap_02_l}
    rotation {pitch=20, yaw=20}
   }
   n24 = Anim Dynamics - Bone: dyn_coat_strap_02_r(ComponentPose=@n23, Alpha=1, ExternalForce="0,0,0") {
    angularDampingOverride=0.998
    angularSpringConstant=500
    bAngularSpring=true
    bOverrideAngularDamping=true
    bUseGravityOverride=true
    bUseSphericalLimits=true
    gravityScale=0.35
    simulationSpace=RootRelative
    alphaBoolBlend {blendOption=Linear}
    boundBone {boneName=dyn_coat_strap_02_r}
    componentAppliedLinearAccClamp=<3000 3000 3000>
    componentLinearAccScale=<0.35 0.35 0.5>
    componentLinearVelScale=<0.35 0.35 0.5>
    gravityOverride {z=-900}
    physicsBodyDefinitions [{
      sphereCollisionRadius=0
      boundBone {boneName=dyn_coat_strap_02_r}
      boxExtents=<15 8 15>
      constraintSetup {angularLimitsMax=<2 20 0>, angularLimitsMin=<-2 0 -35>, angularTarget=<1 0 0>}
      localJointOffset=<5 0 0>
     }]
    sphericalLimits [{limitRadius=7, drivingBone {boneName=thigh_twist_01_r}, sphereLocalOffset=<0 0 0>}]
   }
   n25 = Anim Dynamics - Bone: dyn_coat_strap_03_r(ComponentPose=@n24, Alpha=1, ExternalForce="0,0,0") {
    angularDampingOverride=0.998
    angularSpringConstant=750
    bAngularSpring=true
    bOverrideAngularDamping=true
    bUseGravityOverride=true
    bUseSphericalLimits=true
    gravityScale=0.35
    simulationSpace=RootRelative
    alphaBoolBlend {blendOption=Linear}
    boundBone {boneName=dyn_coat_strap_03_r}
    componentAppliedLinearAccClamp=<3000 3000 3000>
    componentLinearAccScale=<0.35 0.35 0.5>
    componentLinearVelScale=<0.35 0.35 0.5>
    gravityOverride {z=-900}
    physicsBodyDefinitions [{
      sphereCollisionRadius=0
      boundBone {boneName=dyn_coat_strap_03_r}
      boxExtents=<15 8 15>
      constraintSetup {angularLimitsMax=<2 5 0>, angularLimitsMin=<-2 -20 -35>, angularTarget=<1 0 0>}
      localJointOffset=<5 0 0>
     }]
    sphericalLimits [{limitRadius=7, drivingBone {boneName=thigh_twist_01_r}, sphereLocalOffset=<0 0 0>}]
   }
   n26 = Anim Dynamics - Bone: dyn_coat_strap_02_l(ComponentPose=@n25, Alpha=1, ExternalForce="0,0,0") {
    angularDampingOverride=0.998
    angularSpringConstant=500
    bAngularSpring=true
    bOverrideAngularDamping=true
    bUseGravityOverride=true
    bUseSphericalLimits=true
    gravityScale=0.35
    simulationSpace=RootRelative
    alphaBoolBlend {blendOption=Linear}
    boundBone {boneName=dyn_coat_strap_02_l}
    componentAppliedLinearAccClamp=<3000 3000 3000>
    componentLinearAccScale=<0.35 0.35 0.5>
    componentLinearVelScale=<0.35 0.35 0.5>
    gravityOverride {z=-980}
    physicsBodyDefinitions [{
      sphereCollisionRadius=0
      boundBone {boneName=dyn_coat_strap_02_l}
      boxExtents=<12 8 12>
      constraintSetup {angularLimitsMax=<2 0 20>, angularLimitsMin=<-2 -45 -20>, angularTarget=<1 0 0>}
      localJointOffset=<3 0 0>
     }]
    sphericalLimits [
     {limitRadius=7, drivingBone {boneName=thigh_twist_01_l}, sphereLocalOffset=<0 0 2>}
     {limitRadius=7, drivingBone {boneName=thigh_twist_01_l}, sphereLocalOffset=<-6 0 1.5>}
     {limitRadius=7, drivingBone {boneName=calf_l}, sphereLocalOffset=<0 0 1>}
    ]
   }
   n27 = Anim Dynamics - Bone: dyn_coat_strap_03_l(ComponentPose=@n26, Alpha=1, ExternalForce="0,0,0") {
    angularDampingOverride=0.998
    angularSpringConstant=500
    bAngularSpring=true
    bOverrideAngularDamping=true
    bUseGravityOverride=true
    bUseSphericalLimits=true
    gravityScale=0.35
    simulationSpace=RootRelative
    alphaBoolBlend {blendOption=Linear}
    boundBone {boneName=dyn_coat_strap_03_l}
    componentAppliedLinearAccClamp=<3000 3000 3000>
    componentLinearAccScale=<0.35 0.35 0.5>
    componentLinearVelScale=<0.35 0.35 0.5>
    gravityOverride {z=-980}
    physicsBodyDefinitions [{
      sphereCollisionRadius=0
      boundBone {boneName=dyn_coat_strap_03_l}
      boxExtents=<18 12 18>
      constraintSetup {angularLimitsMax=<2 0 20>, angularLimitsMin=<-2 -20 -20>, angularTarget=<1 0 0>}
      localJointOffset=<3 0 0>
     }]
    sphericalLimits [
     {limitRadius=7, drivingBone {boneName=thigh_twist_01_l}, sphereLocalOffset=<0 0 2>}
     {limitRadius=7, drivingBone {boneName=thigh_twist_01_l}, sphereLocalOffset=<-6 0 1.5>}
     {limitRadius=7, drivingBone {boneName=calf_l}, sphereLocalOffset=<0 0 1>}
    ]
   }
   n28 = Get is_melee_move2d_notCrouch() -> (is_melee_move2d_notCrouch:double)
   n29 = "Transform (Modify) Bone - Bone: back_02_r"(ComponentPose=@n27, Alpha=@n28, Translation="0,-4,0", Rotation="0,0,4", Scale="1,1,1") {
    rotationMode=BMM_Additive
    translationMode=BMM_Additive
    translationSpace=BCS_BoneSpace
    alphaBoolBlend {blendOption=Linear}
    boneToModify {boneName=back_02_r}
    rotation {roll=4}
    translation {y=-4}
   }
   n30 = "Transform (Modify) Bone - Bone: back_02_l"(ComponentPose=@n29, Alpha=@n28, Translation="0,3,0", Rotation="0,0,3", Scale="1,1,1") {
    rotationMode=BMM_Additive
    translationMode=BMM_Additive
    translationSpace=BCS_BoneSpace
    alphaBoolBlend {blendOption=Linear}
    boneToModify {boneName=back_02_l}
    rotation {roll=3}
    translation {y=3}
   }
   n31 = "Transform (Modify) Bone - Bone: mid_back_02"(ComponentPose=@n30, Alpha=@n28, Translation="0,-3,0", Rotation="0,0,3", Scale="1,1,1") {
    rotationMode=BMM_Additive
    translationMode=BMM_Additive
    translationSpace=BCS_BoneSpace
    alphaBoolBlend {blendOption=Linear}
    boneToModify {boneName=mid_back_02}
    rotation {roll=3}
    translation {y=-3}
   }
   n32 = Get is_melee_move2d_Crouch() -> (is_melee_move2d_Crouch:double)
   n33 = "Transform (Modify) Bone - Bone: under_arm_02_l"(ComponentPose=@n31, Alpha=@n32, Translation="1.443,3.015,-0.047", Rotation="5.000,0.000,0.000", Scale="1,1,1") {
    rotationMode=BMM_Additive
    translationMode=BMM_Additive
    translationSpace=BCS_BoneSpace
    alphaBoolBlend {blendOption=Linear}
    boneToModify {boneName=under_arm_02_l}
    rotation {pitch=5}
    translation=<1.443 3.015 -0.047>
   }
   n34 = "Transform (Modify) Bone - Bone: frnt_torso_02_l"(ComponentPose=@n33, Alpha=@n32, Translation="0.541,0.834,2.561", Rotation="7,0,-5", Scale="1,1,1") {
    rotationMode=BMM_Additive
    translationMode=BMM_Additive
    translationSpace=BCS_BoneSpace
    alphaBoolBlend {blendOption=Linear}
    boneToModify {boneName=frnt_torso_02_l}
    rotation {pitch=7, roll=-5}
    translation=<0.541 0.834 2.561>
   }
   Output Pose(Result=Component To Local(ComponentPose=@n34)) {layerGroup=None}
   n35 = "Constraint - mid_back_01 (R)"(ComponentPose=Local To Component(LocalPose=Copy Pose From Mesh() {bUseAttachedParent=true}), Alpha=1, ConstraintWeights_0=1) {
    alphaBoolBlend {blendOption=Linear}
    boneToModify {boneName=mid_back_01}
    constraintSetup [{transformType=Rotation, targetBone {boneName=spine_04}}]
    constraintWeights [1]
   }
   n36 = "Constraint - mid_back_02 (R)"(ComponentPose=@n35, Alpha=1, ConstraintWeights_0=1) {
    alphaBoolBlend {blendOption=Linear}
    boneToModify {boneName=mid_back_02}
    constraintSetup [{transformType=Rotation, targetBone {boneName=spine_03}}]
    constraintWeights [1]
   }
   n37 = "Constraint - back_01_r (R)"(ComponentPose=@n36, Alpha=1, ConstraintWeights_0=1) {
    alphaBoolBlend {blendOption=Linear}
    boneToModify {boneName=back_01_r}
    constraintSetup [{transformType=Rotation, targetBone {boneName=spine_04}}]
    constraintWeights [1]
   }
   n38 = "Constraint - back_02_r (R)"(ComponentPose=@n37, Alpha=1, ConstraintWeights_0=1) {
    alphaBoolBlend {blendOption=Linear}
    boneToModify {boneName=back_02_r}
    constraintSetup [{transformType=Rotation, targetBone {boneName=spine_03}}]
    constraintWeights [1]
   }
   n39 = "Constraint - back_01_l (R)"(ComponentPose=@n38, Alpha=1, ConstraintWeights_0=1) {
    alphaBoolBlend {blendOption=Linear}
    boneToModify {boneName=back_01_l}
    constraintSetup [{transformType=Rotation, targetBone {boneName=spine_04}}]
    constraintWeights [1]
   }
   n40 = "Constraint - back_02_l (R)"(ComponentPose=@n39, Alpha=1, ConstraintWeights_0=1) {
    alphaBoolBlend {blendOption=Linear}
    boneToModify {boneName=back_02_l}
    constraintSetup [{transformType=Rotation, targetBone {boneName=spine_03}}]
    constraintWeights [1]
   }
   n41 = Get spine_z_neg() -> (spine_z_neg:double)
   n42 = "Transform (Modify) Bone - Bone: mstr_back_l"(ComponentPose=@n40, Alpha=@n41, Translation="0,-0.75,0", Rotation="0,-5,0", Scale="1,1,1") {
    rotationMode=BMM_Additive
    rotationSpace=BCS_BoneSpace
    translationMode=BMM_Additive
    translationSpace=BCS_BoneSpace
    alphaBoolBlend {blendOption=Linear}
    boneToModify {boneName=mstr_back_l}
    rotation {yaw=-5}
    translation {y=-0.75}
   }
   n43 = "Transform (Modify) Bone - Bone: mstr_mid_back"(ComponentPose=@n42, Alpha=@n41, Translation="0,-1,0", Rotation="0,-6,0", Scale="1,1,1") {
    rotationMode=BMM_Additive
    rotationSpace=BCS_BoneSpace
    translationMode=BMM_Additive
    translationSpace=BCS_BoneSpace
    alphaBoolBlend {blendOption=Linear}
    boneToModify {boneName=mstr_mid_back}
    rotation {yaw=-6}
    translation {y=-1}
   }
   n44 = "Transform (Modify) Bone - Bone: mstr_back_r"(ComponentPose=@n43, Alpha=@n41, Translation="0,-0.75,0", Rotation="0,-6,0", Scale="1,1,1") {
    rotationMode=BMM_Additive
    rotationSpace=BCS_BoneSpace
    translationMode=BMM_Additive
    translationSpace=BCS_BoneSpace
    alphaBoolBlend {blendOption=Linear}
    boneToModify {boneName=mstr_back_r}
    rotation {yaw=-6}
    translation {y=-0.75}
   }
   n45 = "Transform (Modify) Bone - Bone: back_02_l"(ComponentPose=@n44, Alpha=@n41, Translation="0,0,0", Rotation="0,-1,0", Scale="1,1,1") {
    rotationMode=BMM_Additive
    rotationSpace=BCS_BoneSpace
    translationMode=BMM_Additive
    translationSpace=BCS_BoneSpace
    alphaBoolBlend {blendOption=Linear}
    boneToModify {boneName=back_02_l}
    rotation {yaw=-1}
   }
   n46 = "Transform (Modify) Bone - Bone: mid_back_02"(ComponentPose=@n45, Alpha=@n41, Translation="0,0,0", Rotation="0,-1,0", Scale="1,1,1") {
    rotationMode=BMM_Additive
    rotationSpace=BCS_BoneSpace
    translationMode=BMM_Additive
    translationSpace=BCS_BoneSpace
    alphaBoolBlend {blendOption=Linear}
    boneToModify {boneName=mid_back_02}
    rotation {yaw=-1}
   }
   n47 = "Transform (Modify) Bone - Bone: back_02_r"(ComponentPose=@n46, Alpha=@n41, Translation="0,0,0", Rotation="0,-1,0", Scale="1,1,1") {
    rotationMode=BMM_Additive
    rotationSpace=BCS_BoneSpace
    translationMode=BMM_Additive
    translationSpace=BCS_BoneSpace
    alphaBoolBlend {blendOption=Linear}
    boneToModify {boneName=back_02_r}
    rotation {yaw=-1}
   }
   n48 = Control Rig!(Source=Component To Local(ComponentPose=@n47), Alpha, is_crouching:bool=Get isJacketMovedtoSide()) -> (Pose) {
    alphaBoolBlend {blendOption=Linear}
    controlRigAssetReference {blueprintRigClass=&/F_LRG_Leather_Jacket_Purple_Torso_CtrlRig.F_LRG_Leather_Jacket_Purple_Torso_CtrlRig_C}
   }
   Save cached pose 'PreSim Pose Driving'(Pose=@n48) {cacheName=PreSim Pose Driving}
   n49 = Get spineTwistAlpha() -> (spineTwistAlpha:double)
   n50 = "Constraint - dyn_mstr_choker_ring (PPPPP)"!(
    ComponentPose=Local To Component(LocalPose=Use cached pose 'PreSim Pose Driving'())
    Alpha=1
    ConstraintWeights_0=0.3671
    ConstraintWeights_1:float=0.294
    ConstraintWeights_2:float=0.2221
    ConstraintWeights_3:float=0.0684
    ConstraintWeights_4:float=0.0484
   ) -> (Pose) {
    alphaBoolBlend {blendOption=Linear}
    boneToModify {boneName=dyn_mstr_choker_ring}
    constraintSetup [
     {targetBone {boneName=spine_05}}
     {targetBone {boneName=neck_02}}
     {targetBone {boneName=neck_01}}
     {targetBone {boneName=spine_04}}
     {targetBone {boneName=head}}
    ]
    constraintWeights [1,1,1,1,1]
   }
   n51 = "Constraint - mstr_back_d_ring (P)"!(ComponentPose=@n50, Alpha=1, ConstraintWeights_0=1, ConstraintWeights_1:float=0) -> (Pose) {
    alphaBoolBlend {blendOption=Linear}
    boneToModify {boneName=mstr_back_d_ring}
    constraintSetup [{targetBone {boneName=spine_05}}]
    constraintWeights [1,0]
   }
   n52 = "Constraint - necklace_01 (PPPPPPP)"!(
    ComponentPose=@n51
    Alpha=1
    ConstraintWeights_0=0.7222
    ConstraintWeights_1:float=0.1846
    ConstraintWeights_2:float=0.027
    ConstraintWeights_3:float=0.0255
    ConstraintWeights_4:float=0.0207
    ConstraintWeights_5:float=0.02
    ConstraintWeights_6:float=0
   ) -> (Pose) {
    alphaBoolBlend {blendOption=Linear}
    boneToModify {boneName=necklace_01}
    constraintSetup [
     {targetBone {boneName=spine_05}}
     {targetBone {boneName=spine_04}}
     {targetBone {boneName=neck_01}}
     {targetBone {boneName=spine_03}}
     {targetBone {boneName=clavicle_r}}
     {targetBone {boneName=clavicle_l}}
     {targetBone {boneName=head}}
    ]
    constraintWeights [1,1,1,1,1,1,1]
   }
   n53 = "Constraint - mstr_coat_arm_strap_l (PP)"!(ComponentPose=@n52, Alpha=1, ConstraintWeights_0=0.6381, ConstraintWeights_1:float=0.3619) -> (Pose) {
    alphaBoolBlend {blendOption=Linear}
    boneToModify {boneName=mstr_coat_arm_strap_l}
    constraintSetup [{targetBone {boneName=hand_l}}, {targetBone {boneName=dyn_coat_arm_strap_01_l}}]
    constraintWeights [1,1]
   }
   n54 = "Constraint - mstr_coat_arm_strap_r (PP)"!(ComponentPose=@n53, Alpha=1, ConstraintWeights_0=0.6561, ConstraintWeights_1:float=0.3439) -> (Pose) {
    alphaBoolBlend {blendOption=Linear}
    boneToModify {boneName=mstr_coat_arm_strap_r}
    constraintSetup [{targetBone {boneName=dyn_coat_arm_strap_01_r}}, {targetBone {boneName=hand_r}}]
    constraintWeights [1,1]
   }
   n55 = "Constraint - dyn_frnt_coat_r_AIM (PPPPPPPP)"!(
    ComponentPose=@n54
    Alpha=1
    ConstraintWeights_0=0.0906
    ConstraintWeights_1:float=0
    ConstraintWeights_2:float=0.0001
    ConstraintWeights_3:float=0.0151
    ConstraintWeights_4:float=0.2291
    ConstraintWeights_5:float=0
    ConstraintWeights_6:float=0.0459
    ConstraintWeights_7:float=0.6191
   ) -> (Pose) {
    alphaBoolBlend {blendOption=Linear}
    boneToModify {boneName=dyn_frnt_coat_r_AIM}
    constraintSetup [
     {targetBone {boneName=pelvis}}
     {targetBone {boneName=spine_01}}
     {targetBone {boneName=spine_02}}
     {targetBone {boneName=thigh_l}}
     {targetBone {boneName=thigh_r}}
     {targetBone {boneName=dyn_back_03_r}}
     {targetBone {boneName=thigh_twist_01_l}}
     {targetBone {boneName=thigh_twist_01_r}}
    ]
    constraintWeights [1,1,1,1,1,1,1,1]
   }
   n56 = "Constraint - dyn_frnt_coat_l_AIM (PPPPPPPP)"!(
    ComponentPose=@n55
    Alpha=1
    ConstraintWeights_0=0.1253
    ConstraintWeights_1:float=0.0001
    ConstraintWeights_2:float=0.0001
    ConstraintWeights_3:float=0.2569
    ConstraintWeights_4:float=0.0106
    ConstraintWeights_5:float=0.0001
    ConstraintWeights_6:float=0.5745
    ConstraintWeights_7:float=0.0324
   ) -> (Pose) {
    alphaBoolBlend {blendOption=Linear}
    boneToModify {boneName=dyn_frnt_coat_l_AIM}
    constraintSetup [
     {targetBone {boneName=pelvis}}
     {targetBone {boneName=spine_01}}
     {targetBone {boneName=spine_02}}
     {targetBone {boneName=thigh_l}}
     {targetBone {boneName=thigh_r}}
     {targetBone {boneName=dyn_under_arm_03_l}}
     {targetBone {boneName=thigh_twist_01_l}}
     {targetBone {boneName=thigh_twist_01_r}}
    ]
    constraintWeights [1,1,1,1,1,1,1,1]
   }
   n57 = NOT Boolean(A=Get isJacketMovedtoSide())
   n58 = Look At - Bone: frnt_torso_01_l!(ComponentPose=@n56, bAlphaBoolEnabled:bool=@n57, LookAtLocation:Vector=Get L Torso Look At() -> (L Torso Look At:Vector)) -> (Pose) {
    alphaInputType=Bool
    alphaBoolBlend {blendOption=Linear}
    boneToModify {boneName=frnt_torso_01_l}
    lookAtLocation=<3 1 -10>
    lookAtTarget {boneReference {boneName=thigh_l}}
    lookAt_Axis {axis=<1 0>}
   }
   n59 = Look At - Bone: frnt_torso_01_r!(ComponentPose=@n58, bAlphaBoolEnabled:bool=@n57, LookAtLocation:Vector=Get R Torso Look At() -> (R Torso Look At:Vector)) -> (Pose) {
    alphaInputType=Bool
    alphaBoolBlend {blendOption=Linear}
    boneToModify {boneName=frnt_torso_01_r}
    lookAtLocation=<-3 1 -10>
    lookAtTarget {boneReference {boneName=thigh_r}}
    lookAt_Axis {axis=<-1 0>}
   }
   n60 = "Constraint - dyn_mstr_under_arm_r (PPPPPPPP)"!(
    ComponentPose=@n59
    bAlphaBoolEnabled:bool=True
    ConstraintWeights_0=0.4651
    ConstraintWeights_1:float=0.01
    ConstraintWeights_2:float=0.2137
    ConstraintWeights_3:float=0.0161
    ConstraintWeights_4:float=0.0125
    ConstraintWeights_5:float=0.0192
    ConstraintWeights_6:float=0.0133
    ConstraintWeights_7:float=0.25
   ) -> (Pose) {
    alphaInputType=Bool
    bAlphaBoolEnabled=false
    alphaBoolBlend {blendOption=Linear}
    boneToModify {boneName=dyn_mstr_under_arm_r}
    constraintSetup [
     {targetBone {boneName=pelvis}}
     {targetBone {boneName=spine_02}}
     {targetBone {boneName=thigh_r}}
     {targetBone {boneName=dyn_back_01_r}}
     {targetBone {boneName=dyn_back_03_r}}
     {targetBone {boneName=under_arm_02_r}}
     {targetBone {boneName=dyn_under_arm_01_r}}
     {targetBone {boneName=thigh_twist_01_r}}
    ]
    constraintWeights [1,1,1,1,1,1,1,1]
   }
   n61 = "Constraint - dyn_mstr_back_r (PPPPPPPPP)"!(
    ComponentPose=@n60
    bAlphaBoolEnabled:bool=True
    ConstraintWeights_0=0.5304
    ConstraintWeights_1:float=0.1001
    ConstraintWeights_2:float=0.0136
    ConstraintWeights_3:float=0.0338
    ConstraintWeights_4:float=0.0358
    ConstraintWeights_5:float=0.0269
    ConstraintWeights_6:float=0.0248
    ConstraintWeights_7:float=0.01
    ConstraintWeights_8:float=0.2247
   ) -> (Pose) {
    alphaInputType=Bool
    bAlphaBoolEnabled=false
    alphaBoolBlend {blendOption=Linear}
    boneToModify {boneName=dyn_mstr_back_r}
    constraintSetup [
     {targetBone {boneName=pelvis}}
     {targetBone {boneName=thigh_r}}
     {targetBone {boneName=mid_back_02}}
     {targetBone {boneName=dyn_mid_back_01}}
     {targetBone {boneName=dyn_back_01_r}}
     {targetBone {boneName=dyn_back_03_r}}
     {targetBone {boneName=under_arm_02_r}}
     {targetBone {boneName=dyn_under_arm_01_r}}
     {targetBone {boneName=thigh_twist_01_r}}
    ]
    constraintWeights [1,1,1,1,1,1,1,1,1]
   }
   n62 = "Constraint - dyn_mstr_mid_back (PPPPPPPP)"!(
    ComponentPose=@n61
    bAlphaBoolEnabled:bool=True
    ConstraintWeights_0=0.627
    ConstraintWeights_1:float=0.0191
    ConstraintWeights_2:float=0.0602
    ConstraintWeights_3:float=0.0545
    ConstraintWeights_4:float=0.0201
    ConstraintWeights_5:float=0.0239
    ConstraintWeights_6:float=0.0971
    ConstraintWeights_7:float=0.0982
   ) -> (Pose) {
    alphaInputType=Bool
    bAlphaBoolEnabled=false
    alphaBoolBlend {blendOption=Linear}
    boneToModify {boneName=dyn_mstr_mid_back}
    constraintSetup [
     {targetBone {boneName=pelvis}}
     {targetBone {boneName=thigh_r}}
     {targetBone {boneName=mid_back_02}}
     {targetBone {boneName=dyn_mid_back_01}}
     {targetBone {boneName=dyn_mid_back_03}}
     {targetBone {boneName=dyn_back_01_l}}
     {targetBone {boneName=thigh_twist_01_l}}
     {targetBone {boneName=thigh_twist_01_r}}
    ]
    constraintWeights [1,1,1,1,1,1,1,1]
   }
   n63 = "Constraint - dyn_mstr_back_l (PPPPPPPPP)"!(
    ComponentPose=@n62
    bAlphaBoolEnabled:bool=True
    ConstraintWeights_0=0.4738
    ConstraintWeights_1:float=0.0946
    ConstraintWeights_2:float=0.0152
    ConstraintWeights_3:float=0.0345
    ConstraintWeights_4:float=0.0274
    ConstraintWeights_5:float=0.0138
    ConstraintWeights_6:float=0.0336
    ConstraintWeights_7:float=0.0534
    ConstraintWeights_8:float=0.2537
   ) -> (Pose) {
    alphaInputType=Bool
    bAlphaBoolEnabled=false
    alphaBoolBlend {blendOption=Linear}
    boneToModify {boneName=dyn_mstr_back_l}
    constraintSetup [
     {targetBone {boneName=pelvis}}
     {targetBone {boneName=thigh_l}}
     {targetBone {boneName=under_arm_02_l}}
     {targetBone {boneName=dyn_under_arm_01_l}}
     {targetBone {boneName=mid_back_02}}
     {targetBone {boneName=dyn_mid_back_01}}
     {targetBone {boneName=back_02_l}}
     {targetBone {boneName=dyn_back_01_l}}
     {targetBone {boneName=thigh_twist_01_l}}
    ]
    constraintWeights [1,1,1,1,1,1,1,1,1]
   }
   n64 = "Constraint - dyn_mstr_under_arm_l (PPPPPPPP)"!(
    ComponentPose=@n63
    bAlphaBoolEnabled:bool=True
    ConstraintWeights_0=0.5024
    ConstraintWeights_1:float=0.0048
    ConstraintWeights_2:float=0.0088
    ConstraintWeights_3:float=0.201
    ConstraintWeights_4:float=0.0034
    ConstraintWeights_5:float=0.0049
    ConstraintWeights_6:float=0.0073
    ConstraintWeights_7:float=0.2672
   ) -> (Pose) {
    alphaInputType=Bool
    bAlphaBoolEnabled=false
    alphaBoolBlend {blendOption=Linear}
    boneToModify {boneName=dyn_mstr_under_arm_l}
    constraintSetup [
     {targetBone {boneName=pelvis}}
     {targetBone {boneName=spine_01}}
     {targetBone {boneName=spine_02}}
     {targetBone {boneName=thigh_l}}
     {targetBone {boneName=under_arm_02_l}}
     {targetBone {boneName=dyn_under_arm_01_l}}
     {targetBone {boneName=dyn_back_01_l}}
     {targetBone {boneName=thigh_twist_01_l}}
    ]
    constraintWeights [1,1,1,1,1,1,1,1]
   }
   n65 = Get bIsSkyDiving()
   n66 = "Transform (Modify) Bone - Bone: mid_back_01"!(ComponentPose=@n64, bAlphaBoolEnabled:bool=@n65, Translation="0,-1,0", Rotation="0,-3,0", Scale="1,1,1") -> (Pose) {
    alphaInputType=Bool
    rotationMode=BMM_Additive
    rotationSpace=BCS_BoneSpace
    translationMode=BMM_Additive
    translationSpace=BCS_BoneSpace
    alphaBoolBlend {blendOption=Linear}
    boneToModify {boneName=mid_back_01}
    rotation {yaw=-3}
    translation {y=-1}
   }
   n67 = "Transform (Modify) Bone - Bone: back_01_l"!(ComponentPose=@n66, bAlphaBoolEnabled:bool=@n65, Translation="0,-0.5,0", Rotation="0,-2,0", Scale="1,1,1") -> (Pose) {
    alphaInputType=Bool
    rotationMode=BMM_Additive
    rotationSpace=BCS_BoneSpace
    translationMode=BMM_Additive
    translationSpace=BCS_BoneSpace
    alphaBoolBlend {blendOption=Linear}
    boneToModify {boneName=back_01_l}
    rotation {yaw=-2}
    translation {y=-0.5}
   }
   n68 = "Transform (Modify) Bone - Bone: back_01_r"!(ComponentPose=@n67, bAlphaBoolEnabled:bool=@n65, Translation="0,-0.5,0", Rotation="0,-2,0", Scale="1,1,1") -> (Pose) {
    alphaInputType=Bool
    rotationMode=BMM_Additive
    rotationSpace=BCS_BoneSpace
    translationMode=BMM_Additive
    translationSpace=BCS_BoneSpace
    alphaBoolBlend {blendOption=Linear}
    boneToModify {boneName=back_01_r}
    rotation {yaw=-2}
    translation {y=-0.5}
   }
   n69 = "Transform (Modify) Bone - Bone: frnt_torso_01_l"(ComponentPose=@n68, Alpha=@n49, Translation="0,1.5,1", Rotation="0,7,-20", Scale="1,1,1") {
    rotationMode=BMM_Additive
    rotationSpace=BCS_BoneSpace
    translationMode=BMM_Additive
    translationSpace=BCS_BoneSpace
    alphaBoolBlend {blendOption=Linear}
    boneToModify {boneName=frnt_torso_01_l}
    rotation {roll=-20, yaw=7}
    translation {y=1.5, z=1}
   }
   n70 = "Transform (Modify) Bone - Bone: under_arm_01_l"(ComponentPose=@n69, Alpha=@n49, Translation="0,1,0", Rotation="0,5,0", Scale="1,1,1") {
    rotationMode=BMM_Additive
    rotationSpace=BCS_BoneSpace
    translationMode=BMM_Additive
    translationSpace=BCS_BoneSpace
    alphaBoolBlend {blendOption=Linear}
    boneToModify {boneName=under_arm_01_l}
    rotation {yaw=5}
    translation {y=1}
   }
   n71 = "Transform (Modify) Bone - Bone: frnt_torso_01_r"(ComponentPose=@n70, Alpha=@n49, Translation="0,-1,-1", Rotation="7,7,-20", Scale="1,1,1") {
    rotationMode=BMM_Additive
    rotationSpace=BCS_BoneSpace
    translationMode=BMM_Additive
    translationSpace=BCS_BoneSpace
    alphaBoolBlend {blendOption=Linear}
    boneToModify {boneName=frnt_torso_01_r}
    rotation=<rot 7 7 -20>
    translation {y=-1, z=-1}
   }
   n72 = "Transform (Modify) Bone - Bone: under_arm_01_r"(ComponentPose=@n71, Alpha=@n49, Translation="0,-1,0", Rotation="0,5,0", Scale="1,1,1") {
    rotationMode=BMM_Additive
    rotationSpace=BCS_BoneSpace
    translationMode=BMM_Additive
    translationSpace=BCS_BoneSpace
    alphaBoolBlend {blendOption=Linear}
    boneToModify {boneName=under_arm_01_r}
    rotation {yaw=5}
    translation {y=-1}
   }
   n73 = "Transform (Modify) Bone - Bone: frnt_torso_01_l"(
    ComponentPose=@n72
    Alpha=Get spineTwistXTREMEAlpha() -> (spineTwistXTREMEAlpha:double)
    Translation="0,0,0"
    Rotation="15,0,0"
    Scale="1,1,1"
   ) {
    rotationMode=BMM_Additive
    rotationSpace=BCS_BoneSpace
    translationMode=BMM_Additive
    translationSpace=BCS_BoneSpace
    alphaBoolBlend {blendOption=Linear}
    alphaScaleBiasClamp {bInterpResult=true, interpSpeedDecreasing=7}
    boneToModify {boneName=frnt_torso_01_l}
    rotation {pitch=15}
   }
   n74 = "Transform (Modify) Bone - Bone: frnt_torso_01_l"(
    ComponentPose=@n73
    Alpha=Get spineDownXTREMEAlpha() -> (spineDownXTREMEAlpha:double)
    Translation="0,0,0"
    Rotation="20,-15,0"
    Scale="1,1,1"
   ) {
    rotationMode=BMM_Additive
    rotationSpace=BCS_BoneSpace
    translationMode=BMM_Additive
    translationSpace=BCS_BoneSpace
    alphaBoolBlend {blendOption=Linear}
    alphaScaleBiasClamp {bInterpResult=true, interpSpeedDecreasing=7}
    boneToModify {boneName=frnt_torso_01_l}
    rotation {pitch=20, yaw=-15}
   }
   Save cached pose 'PreSim PostDriven Constraints Cache'(Pose=Component To Local(ComponentPose=@n74)) {cacheName=PreSim PostDriven Constraints Cache}
   n75 = "Constraint - under_arm_02_l (P)"!(
    ComponentPose=Local To Component(LocalPose=Use cached pose 'PreSim PostDriven Constraints Cache'())
    bAlphaBoolEnabled:bool=True
    ConstraintWeights_0=1
   ) -> (Pose) {
    alphaInputType=Bool
    bAlphaBoolEnabled=false
    alphaBoolBlend {blendOption=Linear}
    boneToModify {boneName=under_arm_02_l}
    constraintSetup [{targetBone {boneName=spine_03}}]
    constraintWeights [1]
   }
   n76 = "Constraint - under_arm_02_r (P)"!(ComponentPose=@n75, bAlphaBoolEnabled:bool=True, ConstraintWeights_0=1) -> (Pose) {
    alphaInputType=Bool
    bAlphaBoolEnabled=false
    alphaBoolBlend {blendOption=Linear}
    boneToModify {boneName=under_arm_02_r}
    constraintSetup [{targetBone {boneName=spine_03}}]
    constraintWeights [1]
   }
   n77 = "Constraint - frnt_torso_02_l (P)"!(ComponentPose=@n76, bAlphaBoolEnabled:bool=True, ConstraintWeights_0=1) -> (Pose) {
    alphaInputType=Bool
    bAlphaBoolEnabled=false
    alphaBoolBlend {blendOption=Linear}
    boneToModify {boneName=frnt_torso_02_l}
    constraintSetup [{targetBone {boneName=thigh_l}}]
    constraintWeights [1]
   }
   n78 = "Constraint - frnt_torso_02_r (P)"!(ComponentPose=@n77, bAlphaBoolEnabled:bool=True, ConstraintWeights_0=1) -> (Pose) {
    alphaInputType=Bool
    bAlphaBoolEnabled=false
    alphaBoolBlend {blendOption=Linear}
    boneToModify {boneName=frnt_torso_02_r}
    constraintSetup [{targetBone {boneName=spine_01}}]
    constraintWeights [1]
   }
   n79 = "Constraint - dyn_mstr_under_arm_l (P)"!(ComponentPose=@n78, bAlphaBoolEnabled:bool=True, ConstraintWeights_0=1) -> (Pose) {
    alphaInputType=Bool
    bAlphaBoolEnabled=false
    alphaBoolBlend {blendOption=Linear}
    boneToModify {boneName=dyn_mstr_under_arm_l}
    constraintSetup [{targetBone {boneName=spine_01}}]
    constraintWeights [1]
   }
   n80 = "Constraint - dyn_mstr_under_arm_r (P)"!(ComponentPose=@n79, bAlphaBoolEnabled:bool=True, ConstraintWeights_0=1) -> (Pose) {
    alphaInputType=Bool
    bAlphaBoolEnabled=false
    alphaBoolBlend {blendOption=Linear}
    boneToModify {boneName=dyn_mstr_under_arm_r}
    constraintSetup [{targetBone {boneName=spine_01}}]
    constraintWeights [1]
   }
   n81 = "Constraint - dyn_mstr_frnt_torso_l (P)"!(ComponentPose=@n80, bAlphaBoolEnabled:bool=True, ConstraintWeights_0=1) -> (Pose) {
    alphaInputType=Bool
    bAlphaBoolEnabled=false
    alphaBoolBlend {blendOption=Linear}
    boneToModify {boneName=dyn_mstr_frnt_torso_l}
    constraintSetup [{targetBone {boneName=pelvis}}]
    constraintWeights [1]
   }
   n82 = "Constraint - dyn_mstr_frnt_torso_r (P)"!(ComponentPose=@n81, bAlphaBoolEnabled:bool=True, ConstraintWeights_0=1) -> (Pose) {
    alphaInputType=Bool
    bAlphaBoolEnabled=false
    alphaBoolBlend {blendOption=Linear}
    boneToModify {boneName=dyn_mstr_frnt_torso_r}
    constraintSetup [{targetBone {boneName=pelvis}}]
    constraintWeights [1]
   }
   n83 = "Transform (Modify) Bone - Bone: frnt_torso_01_l"!(ComponentPose=@n82, bAlphaBoolEnabled:bool=True, Translation="0,0,0", Rotation="35,-30,-30", Scale="1,1,1") -> (Pose) {
    alphaInputType=Bool
    bAlphaBoolEnabled=false
    rotationMode=BMM_Additive
    rotationSpace=BCS_BoneSpace
    translationMode=BMM_Additive
    translationSpace=BCS_BoneSpace
    alphaBoolBlend {blendOption=Linear}
    boneToModify {boneName=frnt_torso_01_l}
    rotation=<rot 35 -30 -30>
   }
   n84 = "Transform (Modify) Bone - Bone: under_arm_01_l"!(ComponentPose=@n83, bAlphaBoolEnabled:bool=True, Translation="0,0,0", Rotation="-5,10,0", Scale="1,1,1") -> (Pose) {
    alphaInputType=Bool
    bAlphaBoolEnabled=false
    rotationMode=BMM_Additive
    rotationSpace=BCS_BoneSpace
    translationMode=BMM_Additive
    translationSpace=BCS_BoneSpace
    alphaBoolBlend {blendOption=Linear}
    boneToModify {boneName=under_arm_01_l}
    rotation {pitch=-5, yaw=10}
   }
   n85 = RigidBody!(
    ComponentPose=@n84
    Alpha=Get skydiveAlpha() -> (skydiveAlpha:double)
    OverrideWorldGravity:Vector=Get gravity_override() -> (gravity_override:Vector)
    ExternalForce
   ) -> (Pose) {
    bClampLinearTranslationLimitToRefPose=true
    bOverrideWorldGravity=true
    overridePhysicsAsset=&/F_LRG_Leather_Jacket_Purple_Coat_Phys1
    simulationSpace=BaseBoneSpace
    alphaBoolBlend {blendOption=Linear}
    baseBoneRef {boneName=spine_04}
    componentAppliedLinearAccClamp=<5000 5000 5000>
    componentLinearAccScale=<0.3 0.3 0.4>
    componentLinearVelScale=<0.15 0.15 0.15>
   }
   n86 = "Transform (Modify) Bone - Bone: frnt_torso_01_l"!(
    ComponentPose=Local To Component(LocalPose=Use cached pose 'PreSim PostDriven Constraints Cache'())
    Alpha:float=1
    Translation="-3,1,6.5"
    Rotation="0,0,-75"
    Scale="1,1,1"
   ) -> (Pose) {
    rotationMode=BMM_Additive
    rotationSpace=BCS_BoneSpace
    translationMode=BMM_Additive
    translationSpace=BCS_BoneSpace
    alphaBoolBlend {blendOption=Linear}
    boneToModify {boneName=frnt_torso_01_l}
    rotation {roll=-75}
    translation=<-3 1 6.5>
   }
   n87 = Get bIsCrouching()
   n88 = "Constraint - dyn_back_01_l (PP)"!(ComponentPose=@n86, bAlphaBoolEnabled:bool=@n87, ConstraintWeights_0=0.9, ConstraintWeights_1:float=0.1) -> (Pose) {
    alphaInputType=Bool
    bAlphaBoolEnabled=false
    alphaBoolBlend {blendInTime=0.2, blendOption=Linear, blendOutTime=0.2}
    boneToModify {boneName=dyn_back_01_l}
    constraintSetup [{targetBone {boneName=spine_01}}, {targetBone {boneName=dyn_under_arm_01_l}}]
    constraintWeights [1,1]
   }
   n89 = "Constraint - dyn_back_02_l (PP)"!(ComponentPose=@n88, bAlphaBoolEnabled:bool=@n87, ConstraintWeights_0=0.9, ConstraintWeights_1:float=0.1) -> (Pose) {
    alphaInputType=Bool
    bAlphaBoolEnabled=false
    alphaBoolBlend {blendInTime=0.2, blendOption=Linear, blendOutTime=0.2}
    boneToModify {boneName=dyn_back_02_l}
    constraintSetup [{targetBone {boneName=spine_01}}, {targetBone {boneName=dyn_under_arm_02_l}}]
    constraintWeights [1,1]
   }
   n90 = "Constraint - dyn_back_03_l (PP)"!(ComponentPose=@n89, bAlphaBoolEnabled:bool=@n87, ConstraintWeights_0=0.9, ConstraintWeights_1:float=0.1) -> (Pose) {
    alphaInputType=Bool
    bAlphaBoolEnabled=false
    alphaBoolBlend {blendInTime=0.2, blendOption=Linear, blendOutTime=0.2}
    boneToModify {boneName=dyn_back_03_l}
    constraintSetup [{targetBone {boneName=spine_01}}, {targetBone {boneName=dyn_under_arm_03_l}}]
    constraintWeights [1,1]
   }
   n91 = "Constraint - dyn_back_01_r (PP)"!(ComponentPose=@n90, bAlphaBoolEnabled:bool=@n87, ConstraintWeights_0=0.9, ConstraintWeights_1:float=0.1) -> (Pose) {
    alphaInputType=Bool
    bAlphaBoolEnabled=false
    alphaBoolBlend {blendInTime=0.2, blendOption=Linear, blendOutTime=0.2}
    boneToModify {boneName=dyn_back_01_r}
    constraintSetup [{targetBone {boneName=spine_01}}, {targetBone {boneName=dyn_under_arm_01_r}}]
    constraintWeights [1,1]
   }
   n92 = "Constraint - dyn_back_02_r (PP)"!(ComponentPose=@n91, bAlphaBoolEnabled:bool=@n87, ConstraintWeights_0=0.9, ConstraintWeights_1:float=0.1) -> (Pose) {
    alphaInputType=Bool
    bAlphaBoolEnabled=false
    alphaBoolBlend {blendInTime=0.2, blendOption=Linear, blendOutTime=0.2}
    boneToModify {boneName=dyn_back_02_r}
    constraintSetup [{targetBone {boneName=spine_01}}, {targetBone {boneName=dyn_under_arm_02_r}}]
    constraintWeights [1,1]
   }
   n93 = "Constraint - dyn_back_03_r (PP)"!(ComponentPose=@n92, bAlphaBoolEnabled:bool=@n87, ConstraintWeights_0=0.9, ConstraintWeights_1:float=0.1) -> (Pose) {
    alphaInputType=Bool
    bAlphaBoolEnabled=false
    alphaBoolBlend {blendInTime=0.2, blendOption=Linear, blendOutTime=0.2}
    boneToModify {boneName=dyn_back_03_r}
    constraintSetup [{targetBone {boneName=spine_01}}, {targetBone {boneName=dyn_under_arm_03_r}}]
    constraintWeights [1,1]
   }
   n94 = "Constraint - dyn_mid_back_01 (PP)"!(ComponentPose=@n93, bAlphaBoolEnabled:bool=@n87, ConstraintWeights_0=1, ConstraintWeights_1:float=0) -> (Pose) {
    alphaInputType=Bool
    bAlphaBoolEnabled=false
    alphaBoolBlend {blendInTime=0.2, blendOption=Linear, blendOutTime=0.2}
    boneToModify {boneName=dyn_mid_back_01}
    constraintSetup [{targetBone {boneName=spine_01}}, {targetBone {boneName=dyn_mstr_mid_back}}]
    constraintWeights [1,1]
   }
   n95 = "Constraint - dyn_mid_back_02 (PP)"!(ComponentPose=@n94, bAlphaBoolEnabled:bool=@n87, ConstraintWeights_0=1, ConstraintWeights_1:float=0) -> (Pose) {
    alphaInputType=Bool
    bAlphaBoolEnabled=false
    alphaBoolBlend {blendInTime=0.2, blendOption=Linear, blendOutTime=0.2}
    boneToModify {boneName=dyn_mid_back_02}
    constraintSetup [{targetBone {boneName=spine_01}}, {targetBone {boneName=dyn_mid_back_01}}]
    constraintWeights [1,1]
   }
   n96 = "Constraint - dyn_mid_back_03 (PP)"!(ComponentPose=@n95, bAlphaBoolEnabled:bool=@n87, ConstraintWeights_0=1, ConstraintWeights_1:float=0) -> (Pose) {
    alphaInputType=Bool
    bAlphaBoolEnabled=false
    alphaBoolBlend {blendInTime=0.2, blendOption=Linear, blendOutTime=0.2}
    boneToModify {boneName=dyn_mid_back_03}
    constraintSetup [{targetBone {boneName=spine_01}}, {targetBone {boneName=dyn_mid_back_02}}]
    constraintWeights [1,1]
   }
   n97 = RigidBody!(ComponentPose=@n96, bAlphaBoolEnabled:bool, OverrideWorldGravity=Get gravity_override(), ExternalForce) -> (Pose) {
    alphaInputType=Bool
    bClampLinearTranslationLimitToRefPose=true
    bOverrideWorldGravity=true
    overridePhysicsAsset=&/F_LRG_Leather_Jacket_Purple_Coat_Phys1_Crouch
    simulationSpace=BaseBoneSpace
    alphaBoolBlend {blendOption=Linear}
    baseBoneRef {boneName=pelvis}
    componentAppliedLinearAccClamp=<5000 5000 5000>
    componentLinearAccScale=<0.3 0.3 0.4>
    componentLinearVelScale=<0.15 0.15 0.15>
   }
   n98 = "Constraint - dyn_under_arm_01_l (P)"(ComponentPose=@n97, Alpha=0.3, ConstraintWeights_0=1) {
    alphaBoolBlend {blendInTime=0.2, blendOption=Linear, blendOutTime=0.2}
    boneToModify {boneName=dyn_under_arm_01_l}
    constraintSetup [{targetBone {boneName=dyn_frnt_torso_01_l}}]
    constraintWeights [1]
   }
   n99 = "Constraint - dyn_under_arm_02_l (P)"(ComponentPose=@n98, Alpha=0.4, ConstraintWeights_0=1) {
    alphaBoolBlend {blendInTime=0.2, blendOption=Linear, blendOutTime=0.2}
    boneToModify {boneName=dyn_under_arm_02_l}
    constraintSetup [{targetBone {boneName=dyn_frnt_torso_02_l}}]
    constraintWeights [1]
   }
   n100 = "Constraint - dyn_under_arm_03_l (P)"(ComponentPose=@n99, Alpha=0.5, ConstraintWeights_0=1) {
    alphaBoolBlend {blendInTime=0.2, blendOption=Linear, blendOutTime=0.2}
    boneToModify {boneName=dyn_under_arm_03_l}
    constraintSetup [{targetBone {boneName=dyn_frnt_torso_03_l}}]
    constraintWeights [1]
   }
   n101 = Blend Poses by bool(
    BlendPose_0=Component To Local(ComponentPose=@n100)
    BlendPose_1=Component To Local(ComponentPose=@n85)
    BlendTime_0=0.075
    BlendTime_1=0.03
    bActiveValue=Get isJacketMovedtoSide()
   ) {blendType=Linear}
   n102 = "Transform (Modify) Bone - Bone: frnt_torso_01_r"(
    ComponentPose=Local To Component(LocalPose=Use cached pose 'PreSim PostDriven Constraints Cache'())
    Alpha=1
    Translation="3,0,-0.5"
    Rotation="3,3,0"
    Scale="1,1,1"
   ) {
    rotationMode=BMM_Additive
    rotationSpace=BCS_BoneSpace
    translationMode=BMM_Additive
    translationSpace=BCS_BoneSpace
    alphaBoolBlend {blendOption=Linear}
    boneToModify {boneName=frnt_torso_01_r}
    rotation {pitch=3, yaw=3}
    translation {x=3, z=-0.5}
   }
   n103 = "Transform (Modify) Bone - Bone: frnt_torso_01_l"(ComponentPose=@n102, Alpha=1, Translation="-3,0,0.5", Rotation="3,3,0", Scale="1,1,1") {
    rotationMode=BMM_Additive
    rotationSpace=BCS_BoneSpace
    translationMode=BMM_Additive
    translationSpace=BCS_BoneSpace
    alphaBoolBlend {blendOption=Linear}
    boneToModify {boneName=frnt_torso_01_l}
    rotation {pitch=3, yaw=3}
    translation {x=-3, z=0.5}
   }
   n104 = RigidBody(ComponentPose=@n103, OverrideWorldGravity=Get gravity_override()) {
    alphaInputType=Bool
    bClampLinearTranslationLimitToRefPose=true
    bOverrideWorldGravity=true
    overridePhysicsAsset=&/F_LRG_Leather_Jacket_Purple_Coat_Phys1_emote
    simulationSpace=BaseBoneSpace
    alphaBoolBlend {blendOption=Linear}
    baseBoneRef {boneName=pelvis}
    componentAppliedLinearAccClamp=<5000 5000 5000>
    componentLinearAccScale=<0.3 0.3 0.3>
    componentLinearVelScale=<0.15 0.15 0.15>
   }
   n105 = "Constraint - dyn_under_arm_01_r (P)"(ComponentPose=@n104, Alpha=0.3, ConstraintWeights_0=1) {
    alphaBoolBlend {blendInTime=0.2, blendOption=Linear, blendOutTime=0.2}
    boneToModify {boneName=dyn_under_arm_01_r}
    constraintSetup [{targetBone {boneName=dyn_frnt_torso_01_r}}]
    constraintWeights [1]
   }
   n106 = "Constraint - dyn_under_arm_02_r (P)"(ComponentPose=@n105, Alpha=0.4, ConstraintWeights_0=1) {
    alphaBoolBlend {blendInTime=0.2, blendOption=Linear, blendOutTime=0.2}
    boneToModify {boneName=dyn_under_arm_02_r}
    constraintSetup [{targetBone {boneName=dyn_frnt_torso_02_r}}]
    constraintWeights [1]
   }
   n107 = "Constraint - dyn_under_arm_03_r (P)"(ComponentPose=@n106, Alpha=0.5, ConstraintWeights_0=1) {
    alphaBoolBlend {blendInTime=0.2, blendOption=Linear, blendOutTime=0.2}
    boneToModify {boneName=dyn_under_arm_03_r}
    constraintSetup [{targetBone {boneName=dyn_frnt_torso_03_r}}]
    constraintWeights [1]
   }
   n108 = "Constraint - dyn_under_arm_01_l (P)"(ComponentPose=@n107, Alpha=0.3, ConstraintWeights_0=1) {
    alphaBoolBlend {blendInTime=0.2, blendOption=Linear, blendOutTime=0.2}
    boneToModify {boneName=dyn_under_arm_01_l}
    constraintSetup [{targetBone {boneName=dyn_frnt_torso_01_l}}]
    constraintWeights [1]
   }
   n109 = "Constraint - dyn_under_arm_02_l (P)"(ComponentPose=@n108, Alpha=0.4, ConstraintWeights_0=1) {
    alphaBoolBlend {blendInTime=0.2, blendOption=Linear, blendOutTime=0.2}
    boneToModify {boneName=dyn_under_arm_02_l}
    constraintSetup [{targetBone {boneName=dyn_frnt_torso_02_l}}]
    constraintWeights [1]
   }
   n110 = "Constraint - dyn_under_arm_03_l (P)"(ComponentPose=@n109, Alpha=0.5, ConstraintWeights_0=1) {
    alphaBoolBlend {blendInTime=0.2, blendOption=Linear, blendOutTime=0.2}
    boneToModify {boneName=dyn_under_arm_03_l}
    constraintSetup [{targetBone {boneName=dyn_frnt_torso_03_l}}]
    constraintWeights [1]
   }
   n111 = Component To Local(ComponentPose=@n110)
   n112 = Get bIsPlayingEmote()
   n113 = "Transform (Modify) Bone - Bone: mstr_coat_arm_strap_r"(
    ComponentPose=Local To Component(LocalPose=Blend Poses by bool(BlendPose_0=@n111, BlendPose_1=@n101, BlendTime_0=0.25, bActiveValue=@n112) {blendType=Linear})
    Alpha=Get OneHandMeleeAlpha() -> (OneHandMeleeAlpha:double)
    Translation="0,0,0"
    Rotation="-65,0,-30"
    Scale="1,1,1"
   ) {
    rotationMode=BMM_Additive
    rotationSpace=BCS_BoneSpace
    translationMode=BMM_Additive
    translationSpace=BCS_BoneSpace
    alphaBoolBlend {blendOption=Linear}
    boneToModify {boneName=mstr_coat_arm_strap_r}
    rotation {pitch=-65, roll=-30}
   }
   n114 = Get gravity_override()
   n115 = RigidBody!(ComponentPose=@n113, Alpha:float, OverrideWorldGravity=@n114, ExternalForce) -> (Pose) {
    bClampLinearTranslationLimitToRefPose=true
    bOverrideWorldGravity=true
    overridePhysicsAsset=&/F_LRG_Leather_Jacket_Purple_LeftArm_Physics
    simulationSpace=BaseBoneSpace
    alphaBoolBlend {blendOption=Linear}
    baseBoneRef {boneName=lowerarm_l}
    componentAppliedLinearAccClamp=<3000 3000 3000>
    componentLinearAccScale=<0.3 0.3 0.4>
    componentLinearVelScale=<0.3 0.3 0.4>
   }
   n116 = RigidBody(ComponentPose=@n115, OverrideWorldGravity=@n114) {
    bClampLinearTranslationLimitToRefPose=true
    bOverrideWorldGravity=true
    overridePhysicsAsset=&/F_LRG_Leather_Jacket_Purple_RightArm_Physics
    simulationSpace=BaseBoneSpace
    alphaBoolBlend {blendOption=Linear}
    baseBoneRef {boneName=lowerarm_r}
    componentAppliedLinearAccClamp=<3000 3000 3000>
    componentLinearAccScale=<0.3 0.3 0.4>
    componentLinearVelScale=<0.3 0.3 0.4>
   }
   n117 = "Transform (Modify) Bone - Bone: dyn_backBling_adjuster"(ComponentPose=@n116, Alpha=Get BackBlingSDAlpha() -> (BackBlingSDAlpha:double), Translation="0,-2,0", Rotation="0,15,0", Scale="1,1,1") {
    rotationMode=BMM_Additive
    rotationSpace=BCS_BoneSpace
    translationMode=BMM_Additive
    translationSpace=BCS_BoneSpace
    alphaBoolBlend {blendOption=Linear}
    boneToModify {boneName=dyn_backBling_adjuster}
    rotation {yaw=15}
    translation {y=-2}
   }
   n118 = "Transform (Modify) Bone - Bone: dyn_backBling_adjuster"!(ComponentPose=@n117, bAlphaBoolEnabled:bool=Get HoodDown() -> (HoodDown:bool), Translation="0,-3,0", Rotation="0,0,0", Scale="1,1,1") -> (Pose) {
    alphaInputType=Bool
    translationMode=BMM_Additive
    translationSpace=BCS_BoneSpace
    alphaBoolBlend {blendOption=Linear}
    boneToModify {boneName=dyn_backBling_adjuster}
    translation {y=-3}
   }
   Save cached pose 'RBANCache'(Pose=Component To Local(ComponentPose=@n118)) {cacheName=RBANCache}
   Get bIsCrouching()
   Get bIsCrouching()
  ]}
 EventGraph {nodes [
   n1 = Event BlueprintUpdateAnimation()
   > n2 = Sequence!(execute) -> (then_0, then_1, then_2:Exec, then_3:Exec, then_4:Exec, then_5:Exec, then_6:Exec, then_7:Exec, then_8:Exec, then_9:Exec)
   > n3 = Gravity Override State Function DEPRECATED(CharacterPartAnimInstance=Self-Reference(), GameTimeInSeconds=GetGameTimeInSeconds())
   > n4 = Branch(Condition=Get bIsDBNOCarried())
   > Set gravity_override(gravity_override="0, 0,-980")
   Set gravity_override(execute=@n4.else, gravity_override=@n3)
   n5 = Branch(execute=@n2.then_3, Condition=Get bIsSkyDiving())
   > Set skydiveAlpha(skydiveAlpha:double=0.7) -> (Output_Get:double)
   Set skydiveAlpha(execute=@n5.else, skydiveAlpha=SelectFloat(A=0.5, B=1, bPickA=Get bIsDBNOCarried()))
   n6 = GetOwnerSkeletalMeshComponent()
   n7 = BreakRotator(InRot=BreakTransform(InTransform=GetDeltaTransformFromRefPose(self=@n6, BoneName=spine_01, BaseName=spine_05)).Rotation)
   n8 = MapRangeClamped(Value="Absolute (Float)"(A=@n7), InRangeA=10, InRangeB=35)
   > Set spineTwistAlpha(spineTwistAlpha:double=@n8) -> (Output_Get:double)
   n9 = Get bIsMoving2D()
   > n10 = Branch(Condition=AND Boolean(A=Get isJacketMovedtoSide(), B=NOT Boolean(A=@n9)))
   > Set spineTwistXTREMEAlpha(spineTwistXTREMEAlpha:double=@n8) -> (Output_Get:double)
   > Set spineDownXTREMEAlpha(spineDownXTREMEAlpha:double=MapRangeClamped(Value=@n7.Yaw, InRangeA=0, InRangeB=-20)) -> (Output_Get:double)
   Set spineTwistXTREMEAlpha(execute=@n10.else, spineTwistXTREMEAlpha=0)
   > Set spineDownXTREMEAlpha(spineDownXTREMEAlpha=0)
   Set DeltaTimeX(execute=@n2.then_8, DeltaTimeX:double=@n1.DeltaTimeX) -> (Output_Get:double)
   n11 = BreakTransform(InTransform=GetDeltaTransformFromRefPose(self=GetOwnerSkeletalMeshComponent(), BoneName=spine_01, BaseName=spine_05))
   Set spine_z_neg(execute=@n2.then_1, spine_z_neg:double=MapRangeClamped(Value=BreakRotator(InRot=@n11.Rotation).Yaw, InRangeA=-25, InRangeB=-5)) -> (Output_Get:double)
   n12 = BreakTransform(InTransform=GetDeltaTransformFromRefPose(self=GetOwnerSkeletalMeshComponent(), BoneName=dyn_frnt_torso_03_r, BaseName=dyn_mstr_frnt_torso_r))
   Set R Strap Alpha(execute=@n2.then_4, R Strap Alpha:double=MapRangeClamped(Value=BreakRotator(InRot=@n12.Rotation).Pitch, InRangeA=0, InRangeB=-75)) -> (Output_Get:double)
   n13 = BreakTransform(InTransform=GetDeltaTransformFromRefPose(self=GetOwnerSkeletalMeshComponent(), BoneName=dyn_frnt_torso_03_l, BaseName=dyn_mstr_frnt_torso_l))
   > Set L Strap Alpha(L Strap Alpha:double=MapRangeClamped(Value=BreakRotator(InRot=@n13.Rotation).Pitch, InRangeA=0, InRangeB=-75)) -> (Output_Get:double)
   n14 = Get bIsSkyDiving()
   n15 = Get bIsJumping()
   n16 = Get bIsFalling()
   n17 = Get bIsCrouching()
   n18 = Branch(execute=@n2.then_2, Condition=AND Boolean(A=OR Boolean!(A=@n17, B=@n15, C:bool=@n16) -> (ReturnValue), B=NOT Boolean(A=@n14)))
   > Set isJacketMovedtoSide(isJacketMovedtoSide:bool=true) -> (Output_Get:bool)
   > Set coatStrapAlpha(coatStrapAlpha:double=1) -> (Output_Get:double)
   > Set L Torso Look At(L Torso Look At:Vector="-10,-10,-20") -> (Output_Get:Vector)
   > Set R Torso Look At(R Torso Look At:Vector="10,-3,8") -> (Output_Get:Vector)
   Set isJacketMovedtoSide(execute=@n18.else, isJacketMovedtoSide=false)
   > Set coatStrapAlpha(coatStrapAlpha=0.4)
   n19 = BreakTransform(InTransform=GetDeltaTransformFromRefPose(self=GetOwnerSkeletalMeshComponent(), BoneName=spine_01, BaseName=spine_05))
   n20 = MapRangeClamped(Value="Absolute (Float)"(A=BreakRotator(InRot=@n19.Rotation)), InRangeA=10, InRangeB=45, OutRangeA=-10, OutRangeB=-5)
   > Set!(execute, L Torso Look At_X:double=@n20, L Torso Look At_Y:double=8, L Torso Look At_Z:double=-2) -> (then, Output_Get:Vector)
   > Set!(execute, R Torso Look At_X:double="Absolute (Float)"(A=@n20), R Torso Look At_Y:double=-6, R Torso Look At_Z:double=2) -> (then, Output_Get)
   n21 = Branch(execute=@n2.then_5, Condition=Get isBackBling() -> (isBackBling:bool))
   n22 = Get bIsSkydiveDiveMode()
   n23 = Get bIsSkydiveFloating()
   > Set BackBlingSDAlpha(BackBlingSDAlpha:double=SelectFloat(A=1, B=0.25, bPickA=OR Boolean!(A=@n22, B=@n23) -> (ReturnValue))) -> (Output_Get:double)
   Set BackBlingSDAlpha(execute=@n21.else, BackBlingSDAlpha=0)
   n24 = Branch(execute=@n2.then_9, Condition=AND Boolean!(A=Get bIsPlayingMeleeAnim(), B=Get bIsMoving2D(), C:bool=Get bIsCrouching()) -> (ReturnValue))
   > Set is_melee_move2d_Crouch(is_melee_move2d_Crouch:double=FInterpTo_Constant!(Current:double=Get is_melee_move2d_Crouch(), Target:double=1, DeltaTime:double=Get DeltaTimeX() -> (DeltaTimeX:double), InterpSpeed:double=5) -> (ReturnValue:double)) -> (Output_Get:double)
   Set is_melee_move2d_Crouch(
    execute=@n24.else
    is_melee_move2d_Crouch=FInterpTo_Constant(Current=Get is_melee_move2d_Crouch(), Target=0, DeltaTime=Get DeltaTimeX(), InterpSpeed=5)
   )
   Set OneHandMeleeAlpha(execute=@n2.then_6, OneHandMeleeAlpha:double=0) -> (Output_Get:double)
   > n25 = Switch on EFortWeaponCoreAnimation!(
    execute
    Selection:EFortWeaponCoreAnimation Enum=GetWeaponCoreAnim!(self:Fort Player Animation Interface Object Reference=GetOwnerAnimationInterface!(self:Custom Character Part Anim Instance Object Reference) -> (ReturnValue:Fort Player Animation Interface Object Reference)) -> (ReturnValue:EFortWeaponCoreAnimation Enum)
   ) -> (
    Melee:Exec
    Pistol:Exec
    Shotgun:Exec
    PaperBlueprint:Exec
    Rifle:Exec
    MeleeOneHand:Exec
    MachinePistol:Exec
    RocketLauncher:Exec
    GrenadeLauncher:Exec
    GoingCommando:Exec
    AssaultRifle:Exec
    TacticalShotgun:Exec
    SniperRifle:Exec
    TrapPlacement:Exec
    ShoulderLauncher:Exec
    AbilityDecoTool:Exec
    Crossbow:Exec
    C4:Exec
    RemoteControl:Exec
    DualWield:Exec
    AR_BullPup:Exec
    AR_ForwardGrip:Exec
    MedPackPaddles:Exec
    SMG_P90:Exec
    AR_DrumGun:Exec
    Consumable_Small:Exec
    Consumable_Large:Exec
    Balloon:Exec
    MountedTurret:Exec
    CreativeTool:Exec
    ExplosiveBow:Exec
    AshtonIndigo:Exec
    AshtonChicago:Exec
    MeleeDualWield:Exec
    Unarmed:Exec
   )
   Set OneHandMeleeAlpha(execute=@n25.MeleeOneHand, OneHandMeleeAlpha=1)
   n26 = Get bIsCrouching()
   n27 = Branch(execute=@n2.then_7, Condition=AND Boolean(A=Get bIsPlayingMeleeAnim(), B=Get bIsMoving2D(), C=NOT Boolean(A=@n26)))
   > Set is_melee_move2d_notCrouch(is_melee_move2d_notCrouch:double=FInterpTo_Constant(Current=Get is_melee_move2d_notCrouch(), Target=1, DeltaTime=Get DeltaTimeX(), InterpSpeed=5)) -> (Output_Get:double)
   Set is_melee_move2d_notCrouch(
    execute=@n27.else
    is_melee_move2d_notCrouch=FInterpTo_Constant(Current=Get is_melee_move2d_notCrouch(), Target=0, DeltaTime=Get DeltaTimeX(), InterpSpeed=5)
   )
   Event OnCharacterCustomizationLoaded()
   > n28 = Sequence!(execute) -> (then_0, then_1)
   n29 = CheckForBackbling()
   > Set isBackBling(isBackBling:bool=AND Boolean!(A=@n29, B=NOT Boolean(A=@n29.bIsBackblingACape)) -> (ReturnValue)) -> (Output_Get:bool)
   n30 = Branch(execute=@n28.then_1, Condition=CharacterItemDefinitionHasAnyMetaTag(Tags=Make Array("[0]"=(TagName="Cosmetics.Variant.Property.Stage3"))))
   > Set HoodDown(HoodDown:bool=true) -> (Output_Get:bool)
   Set HoodDown(execute=@n30.else, HoodDown=false)
  ]}
]
variables [
 BackBlingSDAlpha {type=number}
 coatStrapAlpha {type=number}
 DeltaTimeX {type=number}
 gravity_override {type=Vector}
 HoodDown {type=boolean}
 is_melee_move2d_Crouch {type=number}
 is_melee_move2d_notCrouch {type=number}
 isBackBling {type=boolean}
 isJacketMovedtoSide {type=boolean}
 L Strap Alpha {type=number}
 L Torso Look At {type=Vector, default=<3 1 -10>}
 OneHandMeleeAlpha {type=number}
 R Strap Alpha {type=number}
 R Torso Look At {type=Vector, default=<-3 1 -10>}
 skydiveAlpha {type=number}
 spine_y_neg {type=number}
 spine_y_pos {type=number}
 spine_z_neg {type=number}
 spine_z_pos {type=number}
 spineDownXTREMEAlpha {type=number}
 spineTwistAlpha {type=number}
 spineTwistXTREMEAlpha {type=number}
]
