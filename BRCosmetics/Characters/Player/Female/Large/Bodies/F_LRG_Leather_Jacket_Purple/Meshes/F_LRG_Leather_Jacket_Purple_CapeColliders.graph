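# Physics asset of cape colliders for the large female player body F_LRG_Leather_Jacket_Purple.
# Each body is a single capsule fitted to a bone, covering head, neck, spine, pelvis, arms, legs and feet, mostly on twist bones for the limbs.
# Further capsules sit on the mid_back and dyn_mid_back bones.
# It has no constraints, so it defines collision volumes only, with no jointed ragdoll setup.
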PhysicsAsset /BRCosmetics/Characters/Player/Female/Large/Bodies/F_LRG_Leather_Jacket_Purple/Meshes/F_LRG_Leather_Jacket_Purple_CapeColliders
bodies {
 calf_twist_02_l [{shapeType=Capsule, length=40.2069, radius=12.7282, center=<-2.3286 -1.1138 -0.1616>, rotation=<rot -0.0062 87.9534 -89.9636>}]
 calf_twist_02_r [{shapeType=Capsule, length=34.1258, radius=13.5433, center=<-2.362 0.5687 0.1066>, rotation=<rot -0.0147 -90.8481 -89.8927>}]
 dyn_mid_back_01 [{shapeType=Capsule, length=20.7756, radius=17.0923, center=<15.5114 10.6495 0.8367>, rotation=<rot 0 -82.8845 0>}]
 foot_l [{shapeType=Capsule, length=14.4159, radius=9.049, center=<-3.1228 5.8608 -0.0635>, rotation=<rot -0.672 179.2463 -91.4399>}]
 foot_r [{shapeType=Capsule, length=9.6659, radius=10.0488, center=<3.377 -5.7397 0.7141>, rotation=<rot 0.6234 1.1655 -88.5384>}]
 head [{shapeType=Capsule, length=2.0719, radius=17.5, center=<13.9496 -0.0001 0>, rotation=<rot 0 -90 0>}]
 lowerarm_l [{shapeType=Capsule, length=31.2982, radius=9.5686, center=<17.5656 -2.5798 -1.2337>, rotation=<rot 2.7686 -102.2653 -86.1165>}]
 lowerarm_twist_01_r [{shapeType=Capsule, length=30.7452, radius=10.0607, center=<-5.2389 0.8795 1.4646>, rotation=<rot 0 -94.0739 -84.6098>}]
 mid_back_01 [{shapeType=Capsule, length=22.0022, radius=15.3604, center=<8.3147 9.6431 0.0001>, rotation=<rot 0 5 0>}]
 mid_back_02 [{shapeType=Capsule, length=20.2216, radius=16.6328, center=<11.3303 10.9037 0.0001>, rotation=<rot 0 -87 0>}]
 neck_01 [{shapeType=Capsule, length=4.3196, radius=8.1138, rotation=<rot -0.0001 -89.9998 90.3863>}]
 neck_02 [{shapeType=Capsule, length=0.0004, radius=8.0807, rotation=<rot 0 -90 0>}]
 pelvis [{shapeType=Capsule, length=19.0256, radius=18.8423, center=<-10.082 -1.0814 -0.3537>, rotation=<rot -4.9681 -82.8847 -0.5913>}]
 spine_01 [{shapeType=Capsule, length=19.9716, radius=16.8828, center=<0 -0.801 0>, rotation=<rot 0 -90 0>}]
 spine_03 [{shapeType=Capsule, length=22.0022, radius=15.3604, center=<0 -1.9917 0.0001>}]
 spine_04 [{shapeType=Capsule, length=7.2641, radius=15.6748, center=<0 -1.0312 0>}]
 spine_05 [{shapeType=Capsule, length=20.0943, radius=13.5568}]
 thigh_twist_01_l [{shapeType=Capsule, length=30.3336, radius=11.7673, center=<7.9097 -2.3933 -2.0262>, rotation=<rot 0.1146 -82.2321 -88.8882>}]
 thigh_twist_01_r [{shapeType=Capsule, length=26.1993, radius=12.5324, center=<-8.9821 -0.1471 2.2298>, rotation=<rot -1.6368 -85.8033 -85.0445>}]
 upperarm_twist_01_l [{shapeType=Capsule, length=18.4787, radius=8.3924, center=<3.8306 -0.4087 0.4242>, rotation=<rot 0 -82.7612 -89.1886>}]
 upperarm_twist_01_r [{shapeType=Capsule, length=24.919, radius=9.4545, center=<-6.1534 1.6414 0.2233>, rotation=<rot -14.813 -92.5958 -92.3309>}]
}
constraints []
