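# Control Rig for the purple leather jacket on the large female body, applying corrective offsets to hood bones as the head rotates.
# Sequenced spherical pose readers on the head bone weight local translations of back and front hood bones.
# Front_hood_01 offsets add to current transforms, so they stack; back_hood_01 and front_hood_02 offset from initial pose.
# Mirroring pose-reader and driven-key library functions, plus a head_ctrl rotation node, are unused.
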ControlRigBlueprint /BRCosmetics/Characters/Player/Female/Large/Bodies/F_LRG_Leather_Jacket_Purple/Meshes/Parts/F_LRG_Leather_Jacket_Purple_FaceAcc_CtrlRig
graphs [
 ContainedGraph {nodes [
   n1 = RigVMFunctionEntryNode()
   > n2 = &RigUnit_SphericalPoseReader(
    ActiveRegionScaleFactors=(PositiveWidth=3.5,NegativeWidth=0.5,PositiveHeight=3,NegativeHeight=3)
    Debug=(bDrawDebug=false,bDraw2D=false,bDrawLocalAxes=false,DebugScale=25,DebugSegments=20,DebugThickness=0.25)
    FalloffRegionScaleFactors=(PositiveWidth=1,NegativeWidth=1,PositiveHeight=1,NegativeHeight=1)
    RotationOffset=RigVMDispatchNode(Array=@n1.rotationOffset, Index=@n1.index) {template="DISPATCH_RigVMDispatch_ArrayGetAtIndex(in Array,in Index,out Element)"}.Element
    DriverItem=RigVMDispatchNode(
     Values=((Type=None,Name="None"),(Type=None,Name="None"))
     Values.1=/Script/ControlRig.RigUnit_ItemReplace(Old=@n1.l_side_token, New=@n1.r_side_token, Item=@n1.driver) {is=struct RigVMUnitNode}.Result
     Index=@n1.index
     Values.0=@n1.driver
    ) {template="DISPATCH_RigVMDispatch_SelectInt32(in Index,in Values,out Result)"}.Result
    DriverAxis=RigVMDispatchNode(
     Values=((X=0,Y=0,Z=0),(X=0,Y=0,Z=0))
     Values.1=/Script/RigVM.RigVMFunction_MathVectorScale(Factor=-1, Value=@n1.driverAxis) {is=struct RigVMUnitNode, template="Scale::Execute(in Value,in Factor,out Result)"}.Result
     Values.0=@n1.driverAxis
     Index=@n1.index
    ) {template="DISPATCH_RigVMDispatch_SelectInt32(in Index,in Values,out Result)"}.Result
    OptionalParentItem=RigVMDispatchNode(
     Values=((Type=None,Name="None"),(Type=None,Name="None"))
     Values.1=&RigUnit_ItemReplace(Old=@n1.l_side_token, New=@n1.r_side_token, Item=@n1.parentSpace).Result
     Index=@n1.index
     Values.0=@n1.parentSpace
    ) {template="DISPATCH_RigVMDispatch_SelectInt32(in Index,in Values,out Result)"}.Result
    ActiveRegionScaleFactors.PositiveWidth=@n1.positiveWidth
    ActiveRegionSize=@n1.activeRegionSize
    FlipHeightScaling=@n1.flipHeightScaling
    FlipWidthScaling=@n1.flipWidthScaling
    Debug.bDrawDebug=@n1.drawDebug
    FalloffSize=@n1.FalloffSize
    ActiveRegionScaleFactors.NegativeHeight=@n1.negativeHeight
    ActiveRegionScaleFactors.PositiveHeight=@n1.positiveHeight
    ActiveRegionScaleFactors.NegativeWidth=@n1.negativeWidth
   )
   > RigVMFunctionReturnNode(outWeights=@n2.OutputParam, outIndex=@n1.index)
  ]}
 ContainedGraph {nodes [
   n1 = RigVMFunctionEntryNode()
   n2 = /Script/RigVM.RigVMFunction_MathTransformToEulerTransform(Value=@n1.transform) {is=struct RigVMUnitNode, template="Cast::Execute(in Value,out Result)"}
   > &RigUnit_ModifyTransforms(
    ItemToModify=((Item=(Type=Bone),Transform=(Rotation=(X=0,Y=0,Z=0,W=1),Translation=(X=0,Y=0,Z=0),Scale3D=(X=0.6,Y=1,Z=1))))
    ItemToModify.0.Transform.Rotation=/Script/RigVM.RigVMFunction_MathQuaternionFromRotator(Rotator=@n2.Result.Rotation) {is=struct RigVMUnitNode}.Result
    ItemToModify.0.Item=RigVMDispatchNode(
     Values=((Type=None,Name="None"),(Type=None,Name="None"))
     Values.1=&RigUnit_ItemReplace(Item=@n1.item, New=@n1.r_side_token, Old=@n1.l_side_token).Result
     Index=@n1.index
     Values.0=@n1.item
    ) {template="DISPATCH_RigVMDispatch_SelectInt32(in Index,in Values,out Result)"}.Result
    ItemToModify.0.Transform.Translation=RigVMDispatchNode(
     Values=((X=0,Y=0,Z=0),(X=0,Y=0,Z=0))
     Values.1=&RigVMFunction_MathVectorScale(Factor=-1, Value=@n2.Result.Location) {template="Scale::Execute(in Value,in Factor,out Result)"}.Result
     Index=@n1.index
     Values.0=@n2.Result.Location
    ) {template="DISPATCH_RigVMDispatch_SelectInt32(in Index,in Values,out Result)"}.Result
    ItemToModify.0.Transform.Scale3D=@n2.Result.Scale
    Weight=@n1.weights
   )
   > RigVMFunctionReturnNode(outWeights=@n1.weights, outIndex=@n1.index)
  ]}
 RigVMFunctionLibrary {nodes [
   PoseReader_FNC = RigVMCollapseNode(
    ExecuteContext=()
    FalloffSize=0.2
    activeRegionSize=0.1
    driver=(Type=Bone)
    driverAxis=(X=1,Y=0,Z=0)
    l_side_token=_l
    negativeHeight=1
    negativeWidth=1
    parentSpace=(Type=Bone)
    positiveHeight=1
    positiveWidth=1
    r_side_token=_r
    rotationOffset=()
   )
   SetDrivenKey_FNC = RigVMCollapseNode(
    ExecuteContext=()
    item=(Type=Bone)
    l_side_token=_l
    r_side_token=_r
    transform=(Rotation=(X=0,Y=0,Z=0,W=1),Translation=(X=0,Y=0,Z=0),Scale3D=(X=1,Y=1,Z=1))
    weights=1
   )
  ]}
 RigVMModel {nodes [
   RigUnit_BeginExecution = &RigUnit_BeginExecution()
   n1 = /Script/ControlRig.RigUnit_SequenceExecution(ExecuteContext=@RigUnit_BeginExecution.ExecutePin) {is=struct RigVMUnitNode}
   n2 = &RigUnit_SphericalPoseReader(
    ActiveRegionScaleFactors=(PositiveWidth=1,NegativeWidth=1,PositiveHeight=1,NegativeHeight=1)
    ActiveRegionSize=0.05
    Debug=(bDrawDebug=false,bDraw2D=false,bDrawLocalAxes=false,DebugScale=25,DebugSegments=20,DebugThickness=0.25)
    DriverAxis=(X=0,Y=1,Z=0)
    DriverItem=(Type=Bone,Name="head")
    FalloffRegionScaleFactors=(PositiveWidth=1,NegativeWidth=1,PositiveHeight=1,NegativeHeight=1)
    FalloffSize=0.25
    OptionalParentItem=(Type=Bone)
    RotationOffset=(X=90,Y=0,Z=-70)
    ExecutePin=@n1.B
   )
   > /Script/ControlRig.RigUnit_SetTranslation(
    Item=(Type=Bone,Name="back_hood_01")
    Space=LocalSpace
    bPropagateToChildren=True
    Value=/Script/RigVM.RigVMFunction_MathVectorAdd(
     B=(X=6,Y=-2,Z=0)
     A=&RigUnit_GetTransform(Item=(Type=Bone,Name="back_hood_01"), Space=LocalSpace, bInitial=true).Transform.Translation
    ) {is=struct RigVMUnitNode, template="Add::Execute(in A,in B,out Result)"}.Result
    Weight=@n2.OutputParam
   ) {is=struct RigVMUnitNode, template="Set Transform::Execute(in Item,in Space,in bInitial,in Value,in Weight,in bPropagateToChildren)"}
   n3 = &RigUnit_SphericalPoseReader(
    ActiveRegionScaleFactors=(PositiveWidth=1,NegativeWidth=1,PositiveHeight=1,NegativeHeight=1)
    ActiveRegionSize=0.2
    Debug=(bDrawDebug=false,bDraw2D=false,bDrawLocalAxes=false,DebugScale=25,DebugSegments=20,DebugThickness=0.25)
    DriverAxis=(X=0,Y=1,Z=0)
    DriverItem=(Type=Bone,Name="head")
    FalloffRegionScaleFactors=(PositiveWidth=1,NegativeWidth=1,PositiveHeight=1,NegativeHeight=1)
    FalloffSize=0.4
    OptionalParentItem=(Type=Bone)
    RotationOffset=(X=180,Y=0,Z=0)
    ExecutePin=@n1.C
   )
   > &RigUnit_SetTranslation(
    Item=(Type=Bone,Name="front_hood_01_l")
    Space=LocalSpace
    bPropagateToChildren=True
    Value=&RigVMFunction_MathVectorAdd(
     B=(X=-1,Y=1,Z=-1)
     A=&RigUnit_GetTransform(Item=(Type=Bone,Name="front_hood_01_l"), Space=LocalSpace).Transform.Translation
    ) {template="Add::Execute(in A,in B,out Result)"}.Result
    Weight=@n3.OutputParam
   ) {template="Set Transform::Execute(in Item,in Space,in bInitial,in Value,in Weight,in bPropagateToChildren)"}
   > n4 = &RigUnit_SphericalPoseReader(
    ActiveRegionScaleFactors=(PositiveWidth=1,NegativeWidth=1,PositiveHeight=1,NegativeHeight=1)
    ActiveRegionSize=0.2
    Debug=(bDrawDebug=false,bDrawLocalAxes=false,DebugScale=25,DebugSegments=20,DebugThickness=0.25)
    DriverAxis=(X=0,Y=1,Z=0)
    DriverItem=(Type=Bone,Name="head")
    FalloffRegionScaleFactors=(PositiveWidth=1,NegativeWidth=1,PositiveHeight=1,NegativeHeight=1)
    FalloffSize=0.4
    OptionalParentItem=(Type=Bone)
    RotationOffset=(X=0,Y=0,Z=0)
   )
   > &RigUnit_SetTranslation(
    Item=(Type=Bone,Name="front_hood_01_r")
    Space=LocalSpace
    bPropagateToChildren=True
    Value=&RigVMFunction_MathVectorAdd(B=(X=1,Y=-1,Z=1), A=&RigUnit_GetTransform(Item=(Type=Bone,Name="front_hood_01_r"), Space=LocalSpace).Transform.Translation) {template="Add::Execute(in A,in B,out Result)"}.Result
    Weight=@n4.OutputParam
   ) {template="Set Transform::Execute(in Item,in Space,in bInitial,in Value,in Weight,in bPropagateToChildren)"}
   n5 = &RigUnit_SequenceExecution(ExecuteContext=@n1.D)
   n6 = &RigUnit_SphericalPoseReader(
    ActiveRegionScaleFactors=(PositiveWidth=1,NegativeWidth=1,PositiveHeight=1,NegativeHeight=1)
    ActiveRegionSize=0.1
    Debug=(bDrawDebug=false,bDrawLocalAxes=false,DebugScale=25,DebugSegments=20,DebugThickness=0.25)
    DriverAxis=(X=1,Y=0,Z=0)
    DriverItem=(Type=Bone,Name="head")
    FalloffRegionScaleFactors=(PositiveWidth=1,NegativeWidth=1,PositiveHeight=1,NegativeHeight=1)
    FalloffSize=0.2
    OptionalParentItem=(Type=Bone)
    RotationOffset=(X=0,Y=-50,Z=0)
    ExecutePin=@n5.A
   )
   > &RigUnit_SetTranslation(
    Item=(Type=Bone,Name="front_hood_01_l")
    Space=LocalSpace
    bPropagateToChildren=True
    Value=&RigVMFunction_MathVectorAdd(
     B=(X=1,Y=-1,Z=-1)
     A=&RigUnit_GetTransform(Item=(Type=Bone,Name="front_hood_01_l"), Space=LocalSpace).Transform.Translation
    ) {template="Add::Execute(in A,in B,out Result)"}.Result
    Weight=@n6.OutputParam
   ) {template="Set Transform::Execute(in Item,in Space,in bInitial,in Value,in Weight,in bPropagateToChildren)"}
   > n7 = &RigUnit_SphericalPoseReader(
    ActiveRegionScaleFactors=(PositiveWidth=1,NegativeWidth=1,PositiveHeight=1,NegativeHeight=1)
    ActiveRegionSize=0.1
    Debug=(bDrawDebug=false,bDrawLocalAxes=false,DebugScale=25,DebugSegments=20,DebugThickness=0.25)
    DriverAxis=(X=1,Y=0,Z=0)
    DriverItem=(Type=Bone,Name="head")
    FalloffRegionScaleFactors=(PositiveWidth=1,NegativeWidth=1,PositiveHeight=1,NegativeHeight=1)
    FalloffSize=0.2
    OptionalParentItem=(Type=Bone)
    RotationOffset=(X=0,Y=220,Z=0)
   )
   > &RigUnit_SetTranslation(
    Item=(Type=Bone,Name="front_hood_01_r")
    Space=LocalSpace
    bPropagateToChildren=True
    Value=&RigVMFunction_MathVectorAdd(B=(X=1,Y=1,Z=1), A=&RigUnit_GetTransform(Item=(Type=Bone,Name="front_hood_01_r"), Space=LocalSpace).Transform.Translation) {template="Add::Execute(in A,in B,out Result)"}.Result
    Weight=@n7.OutputParam
   ) {template="Set Transform::Execute(in Item,in Space,in bInitial,in Value,in Weight,in bPropagateToChildren)"}
   n8 = &RigUnit_SphericalPoseReader(
    ActiveRegionScaleFactors=(PositiveWidth=1,NegativeWidth=1,PositiveHeight=1,NegativeHeight=1)
    ActiveRegionSize=0.1
    Debug=(bDrawDebug=false,DebugScale=25,DebugSegments=20,DebugThickness=0.25)
    DriverAxis=(X=1,Y=0,Z=0)
    DriverItem=(Type=Bone,Name="head")
    FalloffRegionScaleFactors=(PositiveWidth=1,NegativeWidth=1,PositiveHeight=1,NegativeHeight=1)
    FalloffSize=0.2
    OptionalParentItem=(Type=Bone)
    RotationOffset=(X=0,Y=230,Z=0)
    ExecutePin=@n5.B
   )
   > &RigUnit_SetTranslation(
    Item=(Type=Bone,Name="front_hood_01_l")
    Space=LocalSpace
    bPropagateToChildren=True
    Value=&RigVMFunction_MathVectorAdd(B=(X=5,Y=2,Z=0), A=&RigUnit_GetTransform(Item=(Type=Bone,Name="front_hood_01_l"), Space=LocalSpace).Transform.Translation) {template="Add::Execute(in A,in B,out Result)"}.Result
    Weight=@n8.OutputParam
   ) {template="Set Transform::Execute(in Item,in Space,in bInitial,in Value,in Weight,in bPropagateToChildren)"}
   > &RigUnit_SetTranslation(
    Item=(Type=Bone,Name="front_hood_02_l")
    Space=LocalSpace
    bPropagateToChildren=True
    Value=&RigVMFunction_MathVectorAdd(
     B=(X=5,Y=1.5,Z=-1)
     A=&RigUnit_GetTransform(Item=(Type=Bone,Name="front_hood_02_l"), Space=LocalSpace, bInitial=true).Transform.Translation
    ) {template="Add::Execute(in A,in B,out Result)"}.Result
    Weight=@n8.OutputParam
   ) {template="Set Transform::Execute(in Item,in Space,in bInitial,in Value,in Weight,in bPropagateToChildren)"}
   > n9 = &RigUnit_SphericalPoseReader(
    ActiveRegionScaleFactors=(PositiveWidth=1,NegativeWidth=1,PositiveHeight=1,NegativeHeight=1)
    ActiveRegionSize=0.1
    Debug=(bDrawDebug=false,DebugScale=25,DebugSegments=20,DebugThickness=0.25)
    DriverAxis=(X=1,Y=0,Z=0)
    DriverItem=(Type=Bone,Name="head")
    FalloffRegionScaleFactors=(PositiveWidth=1,NegativeWidth=1,PositiveHeight=1,NegativeHeight=1)
    FalloffSize=0.2
    FlipHeightScaling=false
    FlipWidthScaling=false
    OptionalParentItem=(Type=Bone)
    RotationOffset=(X=0,Y=-52,Z=10)
   )
   > &RigUnit_SetTranslation(
    Item=(Type=Bone,Name="front_hood_01_r")
    Space=LocalSpace
    bPropagateToChildren=True
    Value=&RigVMFunction_MathVectorAdd(B=(X=5,Y=-2,Z=0), A=&RigUnit_GetTransform(Item=(Type=Bone,Name="front_hood_01_r"), Space=LocalSpace).Transform.Translation) {template="Add::Execute(in A,in B,out Result)"}.Result
    Weight=@n9.OutputParam
   ) {template="Set Transform::Execute(in Item,in Space,in bInitial,in Value,in Weight,in bPropagateToChildren)"}
   > &RigUnit_SetTranslation(
    Item=(Type=Bone,Name="front_hood_02_r")
    Space=LocalSpace
    bPropagateToChildren=True
    Value=&RigVMFunction_MathVectorAdd(
     B=(X=5,Y=-1.5,Z=1)
     A=&RigUnit_GetTransform(Item=(Type=Bone,Name="front_hood_02_r"), Space=LocalSpace, bInitial=true).Transform.Translation
    ) {template="Add::Execute(in A,in B,out Result)"}.Result
    Weight=@n9.OutputParam
   ) {template="Set Transform::Execute(in Item,in Space,in bInitial,in Value,in Weight,in bPropagateToChildren)"}
   &RigUnit_SetRotation(
    ExecutePin=()
    Item=(Type=Bone,Name="head")
    Space=GlobalSpace
    Weight=1
    Value=&RigUnit_GetTransform(Item=(Type=Control,Name="head_ctrl")).Transform.Rotation
   )
  ]}
]
hierarchy [
 root {loc=<0 0 0>, rot=<0 0 0 1>}
 pelvis {parent=root, loc=<0 -0.648 94.7699>, rot=<0.0127 0.707 -0.0127 -0.707>}
 spine_01 {parent=pelvis, loc=<5.0975 -0.6388 0>, rot=<0 0 0.034 -0.9994>}
 spine_02 {parent=spine_01, loc=<7.2518 -0.035 0>, rot=<0 0 0.0238 -0.9997>}
 spine_03 {parent=spine_02, loc=<6.0104 0 0>, rot=<0 0 0.0149 -0.9999>}
 spine_04 {parent=spine_03, loc=<8.1024 0 0>, rot=<0 0 -0.0132 -0.9999>}
 spine_05 {parent=spine_04, loc=<10.0146 0 0>, rot=<0 0 -0.0015 -1>}
 clavicle_l {parent=spine_05, loc=<4.1651 -0.3023 -5.0942>, rot=<0.0465 -0.7289 -0.009 -0.6829>}
 clavicle_r {parent=spine_05, loc=<4.1652 -0.3023 5.0942>, rot=<0.7289 0.0465 -0.6829 0.009>}
 neck_01 {parent=spine_05, loc=<11.3775 -0.6396 0>, rot=<0 0 -0.2003 -0.9797>}
 neck_02 {parent=neck_01, loc=<4.5627 0 0>, rot=<0 0 0.0504 -0.9987>}
 head {parent=neck_02, loc=<5.3914 0 0>, rot=<0 0 0.115 -0.9934>}
 dyn_hair_01_r {parent=head, loc=<-0.4051 5.1659 6.3001>, rot=<-0.0876 -0.2452 0.9651 -0.0294>}
 dyn_hair_01_l {parent=head, loc=<-1.5625 4.7458 -5.8846>, rot=<0.0755 0.1438 0.9843 0.0693>}
 dyn_hair_bangs_01 {parent=head, loc=<8.7884 9.8194 1.1506>, rot=<-0.0477 -0.1342 0.9889 0.0431>}
 dyn_hair_bangs_03 {parent=head, loc=<10.6047 8.5082 -2.8794>, rot=<-0.2812 0.531 0.7993 0.0038>}
 dyn_hair_bangs_02 {parent=head, loc=<10.0668 9.2731 -1.129>, rot=<-0.2692 0.0026 0.9629 -0.0174>}
 hood {parent=head, loc=<0 0 0>, rot=<0 0 0 1>}
 back_hood_01 {parent=hood, loc=<-0.6976 -6.7063 0>, rot=<0 0 0 1>}
 front_hood_01_l {parent=hood, loc=<-0.8334 3.7059 -8.5108>, rot=<0 0 0 1>}
 front_hood_01_r {parent=hood, loc=<-0.8339 3.7059 8.5108>, rot=<0 0 1 0>}
 front_hood_02_l {parent=neck_02, loc=<-1.5581 3.2182 -9.1658>, rot=<0 0 0 1>}
 front_hood_02_r {parent=neck_02, loc=<-1.5585 3.2183 9.1658>, rot=<0 0 1 0>}
 head_ctrl {controlType=EULER_TRANSFORM, type=Control, loc=<0 0 0>, rot=<0 0 0 1>}
]
